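# Physics asset for the SummitReedGrit male hat-variant face accessory, giving secondary motion to bandana and hair bones.
# Three-link capsule chains for the center, left and right bandana and front and side hair hang from small MSTR anchor spheres with tight angular limits.
# Extra free-swing constraints tie the bandana tips together and each side hair tip to its front counterpart.
# Neck, upper spine, clavicle and upper-arm shapes provide body collision volumes.
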PhysicsAsset /BRCosmetics/Characters/Player/Male/Medium/Bodies/M_MED_SummitReedGrit_Hat/Meshes/Parts/M_MED_SummitReedGrit_FaceAccHatBackPacked_Physics
bodies {
 MSTR_bandana [{shapeType=Sphere, radius=0.505, shapeName=MSTR_bandana_sphere}]
 MSTR_bandana_l [{shapeType=Sphere, radius=0.505, shapeName=MSTR_bandana_l_sphere}]
 MSTR_bandana_r [{shapeType=Sphere, radius=0.505, shapeName=MSTR_bandana_r_sphere}]
 MSTR_hair_fr_l [{shapeType=Sphere, radius=0.505, shapeName=MSTR_hair_fr_l_sphere}]
 MSTR_hair_fr_r [{shapeType=Sphere, radius=0.505, shapeName=MSTR_hair_fr_r_sphere}]
 MSTR_hair_side_l [{shapeType=Sphere, radius=0.505, shapeName=MSTR_hair_side_l_sphere}]
 MSTR_hair_side_r [{shapeType=Sphere, radius=0.505, shapeName=MSTR_hair_side_r_sphere}]
 clavicle_l [{
   shapeType=Capsule
   length=1.114
   radius=8.329
   shapeName=clavicle_r_capsule
   center=<9.7424 -1.3218 -1.8665>
   rotation=<rot -0.6007 94.6351 97.3669>
  }]
 clavicle_r [{
   shapeType=Capsule
   length=1.114
   radius=8.329
   shapeName=clavicle_r_capsule
   center=<-9.7424 1.3218 1.8665>
   rotation=<rot -0.6007 -85.3649 -82.6331>
  }]
 dyn_bandana_01 [{shapeType=Capsule, length=2.0212, radius=4.2389, shapeName=dyn_bandana_01_capsule, center=<3.7871 -2.484 0>}]
 dyn_bandana_01_l [{
   shapeType=Capsule
   length=2.5818
   radius=3
   shapeName=dyn_bandana_01_l_capsule
   center=<4.2726 -3.6094 -2.3415>
   rotation=<rot 0.0002 0.0022 20.24>
  }]
 dyn_bandana_01_r [{
   shapeType=Capsule
   length=2.5818
   radius=3
   shapeName=dyn_bandana_01_r_capsule
   center=<4.7392 -4.1253 2.1512>
   rotation=<rot -0.0002 0.0022 166.9018>
  }]
 dyn_bandana_02 [{shapeType=Capsule, length=1.7128, radius=5.1785, shapeName=dyn_bandana_02_capsule, center=<3.3503 -2.1055 0>}]
 dyn_bandana_02_l [{
   shapeType=Capsule
   length=2.8318
   radius=3.692
   shapeName=dyn_bandana_02_l_capsule
   center=<4.1435 -3.1894 -2.5328>
   rotation=<rot -2.7989 -1.0306 27.8144>
  }]
 dyn_bandana_02_r [{
   shapeType=Capsule
   length=2.748
   radius=3.7758
   shapeName=dyn_bandana_02_r_capsule
   center=<4.1551 -2.397 2.5863>
   rotation=<rot 2.7989 -1.0305 -5.1591>
  }]
 dyn_bandana_03 [{shapeType=Capsule, length=2.8718, radius=4.2972, shapeName=dyn_bandana_03_capsule, center=<3.3503 -3.1242 0>}]
 dyn_bandana_03_l [{
   shapeType=Capsule
   length=3.6284
   radius=3.1407
   shapeName=dyn_bandana_03_l_capsule
   center=<4.3176 -3.2208 -2.4848>
   rotation=<rot -0.0002 0.0023 14.733>
  }]
 dyn_bandana_03_r [{
   shapeType=Capsule
   length=3.1482
   radius=3.621
   shapeName=dyn_bandana_03_r_capsule
   center=<4.3109 -3.4501 2.4592>
   rotation=<rot 3.425 -3.8122 -6.4701>
  }]
 dyn_hair_01_fr_l [{shapeType=Capsule, length=2.0009, radius=2.6199, shapeName=dyn_hair_01_fr_l_capsule, center=<2.4565 -2.4629 -1.2429>}]
 dyn_hair_01_fr_r [{shapeType=Capsule, length=2.0009, radius=2.6199, shapeName=dyn_hair_01_fr_r_capsule, center=<-2.4565 2.4629 1.2429>, rotation=<rot 0 0 -180>}]
 dyn_hair_01_side_l [{shapeType=Capsule, length=2.0009, radius=2.6199, shapeName=dyn_hair_01_fr_l_capsule, center=<2.4565 -2.4629 -1.6749>}]
 dyn_hair_01_side_r [{shapeType=Capsule, length=2.0009, radius=2.6199, shapeName=dyn_hair_01_fr_r_capsule, center=<-2.4565 2.4629 1.6749>, rotation=<rot 0 0 -180>}]
 dyn_hair_02_fr_l [{shapeType=Capsule, length=3.1233, radius=2.5821, shapeName=dyn_hair_02_fr_l_capsule, center=<1.981 -2.8387 0>}]
 dyn_hair_02_fr_r [{shapeType=Capsule, length=3.1233, radius=2.5821, shapeName=dyn_hair_02_fr_r_capsule, center=<-1.981 2.8387 0>, rotation=<rot 0 0 -180>}]
 dyn_hair_02_side_l [{
   shapeType=Capsule
   length=3.1233
   radius=2.5821
   shapeName=dyn_hair_02_fr_l_capsule
   center=<1.9893 -2.8387 -1.574>
   rotation=<rot -2.7146 0 0>
  }]
 dyn_hair_02_side_r [{
   shapeType=Capsule
   length=3.1233
   radius=2.5821
   shapeName=dyn_hair_02_fr_r_capsule
   center=<-1.9893 2.8387 1.574>
   rotation=<rot -2.7146 0 -180>
  }]
 dyn_hair_03_fr_l [{shapeType=Capsule, length=1.7606, radius=2.5601, shapeName=dyn_hair_03_fr_l_capsule, center=<1.3508 -2.4533 0>}]
 dyn_hair_03_fr_r [{shapeType=Capsule, length=1.7606, radius=2.5601, shapeName=dyn_hair_03_fr_r_capsule, center=<-1.3508 2.4533 0>, rotation=<rot 0 0 -180>}]
 dyn_hair_03_side_l [{
   shapeType=Capsule
   length=1.7606
   radius=2.5601
   shapeName=dyn_hair_03_fr_l_capsule
   center=<0.8537 -2.1272 -1.622>
   rotation=<rot -6.7874 0 0>
  }]
 dyn_hair_03_side_r [{
   shapeType=Capsule
   length=1.7606
   radius=2.5601
   shapeName=dyn_hair_03_fr_r_capsule
   center=<-0.8088 2.1743 1.6166>
   rotation=<rot -6.7874 0 -180>
  }]
 neck_01 [{shapeType=Sphere, radius=5.6639, shapeName=neck_01_sphere, center=<0 0.1605 0>}]
 neck_02 [{shapeType=Sphere, radius=5.3844, shapeName=neck_01_sphere, center=<0 -0.7872 0>}]
 spine_05 [{shapeType=Capsule, length=10.0819, radius=10.0165, shapeName=spine_05_capsule, center=<1.1226 -1.4292 0>, rotation=<rot 0 -90 0>}]
 upperarm_l [{
   shapeType=Capsule
   length=16.1125
   radius=5.8304
   shapeName=clavicle_l_capsule
   center=<9.9246 -1.3218 -0.4619>
   rotation=<rot -0.0002 -89.9999 87.3912>
  }]
 upperarm_r [{
   shapeType=Capsule
   length=16.1125
   radius=5.8304
   shapeName=clavicle_r_capsule
   center=<-9.9246 1.3218 0.4619>
   rotation=<rot -0.0002 -89.9999 -92.6088>
  }]
}
constraints [
 {bone1Name=dyn_bandana_01, bone2Name=MSTR_bandana, swing1LimitDegrees=20, swing2LimitDegrees=20, twistLimitDegrees=5}
 {bone1Name=dyn_bandana_02, bone2Name=dyn_bandana_01, swing1LimitDegrees=20, swing2LimitDegrees=20, twistLimitDegrees=5}
 {bone1Name=dyn_bandana_03, bone2Name=dyn_bandana_02, swing1LimitDegrees=20, swing2LimitDegrees=20, twistLimitDegrees=5}
 {
  bone1Name=dyn_bandana_03
  bone2Name=dyn_bandana_03_r
  swing1LimitDegrees=45
  swing1Motion=Free
  swing2LimitDegrees=45
  swing2Motion=Free
  twistLimitDegrees=45
  twistMotion=Locked
 }
 {
  bone1Name=dyn_bandana_03
  bone2Name=dyn_bandana_03_l
  swing1LimitDegrees=45
  swing1Motion=Free
  swing2LimitDegrees=45
  swing2Motion=Free
  twistLimitDegrees=45
  twistMotion=Locked
 }
 {bone1Name=dyn_bandana_01_l, bone2Name=MSTR_bandana_l, swing1LimitDegrees=20, swing2LimitDegrees=20, twistLimitDegrees=5}
 {bone1Name=dyn_bandana_02_l, bone2Name=dyn_bandana_01_l, swing1LimitDegrees=20, swing2LimitDegrees=20, twistLimitDegrees=5}
 {bone1Name=dyn_bandana_03_l, bone2Name=dyn_bandana_02_l, swing1LimitDegrees=20, swing2LimitDegrees=20, twistLimitDegrees=5}
 {bone1Name=dyn_bandana_01_r, bone2Name=MSTR_bandana_r, swing1LimitDegrees=20, swing2LimitDegrees=20, twistLimitDegrees=5}
 {bone1Name=dyn_bandana_02_r, bone2Name=dyn_bandana_01_r, swing1LimitDegrees=20, swing2LimitDegrees=20, twistLimitDegrees=5}
 {bone1Name=dyn_bandana_03_r, bone2Name=dyn_bandana_02_r, swing1LimitDegrees=20, swing2LimitDegrees=20, twistLimitDegrees=5}
 {bone1Name=dyn_hair_01_fr_l, bone2Name=MSTR_hair_fr_l, swing1LimitDegrees=15, swing2LimitDegrees=10, twistLimitDegrees=0}
 {bone1Name=dyn_hair_02_fr_l, bone2Name=dyn_hair_01_fr_l, swing1LimitDegrees=15, swing2LimitDegrees=10, twistLimitDegrees=5}
 {bone1Name=dyn_hair_03_fr_l, bone2Name=dyn_hair_02_fr_l, swing1LimitDegrees=20, swing2LimitDegrees=10, twistLimitDegrees=5}
 {
  bone1Name=dyn_hair_03_side_l
  bone2Name=dyn_hair_03_fr_l
  swing1LimitDegrees=45
  swing1Motion=Free
  swing2LimitDegrees=45
  swing2Motion=Free
  twistLimitDegrees=45
  twistMotion=Free
 }
 {bone1Name=dyn_hair_01_fr_r, bone2Name=MSTR_hair_fr_r, swing1LimitDegrees=15, swing2LimitDegrees=10, twistLimitDegrees=0}
 {bone1Name=dyn_hair_02_fr_r, bone2Name=dyn_hair_01_fr_r, swing1LimitDegrees=15, swing2LimitDegrees=10, twistLimitDegrees=5}
 {bone1Name=dyn_hair_03_fr_r, bone2Name=dyn_hair_02_fr_r, swing1LimitDegrees=20, swing2LimitDegrees=10, twistLimitDegrees=5}
 {
  bone1Name=dyn_hair_03_side_r
  bone2Name=dyn_hair_03_fr_r
  swing1LimitDegrees=45
  swing1Motion=Free
  swing2LimitDegrees=45
  swing2Motion=Free
  twistLimitDegrees=45
  twistMotion=Free
 }
 {bone1Name=dyn_hair_01_side_l, bone2Name=MSTR_hair_side_l, swing1LimitDegrees=15, swing2LimitDegrees=10, twistLimitDegrees=0}
 {bone1Name=dyn_hair_02_side_l, bone2Name=dyn_hair_01_side_l, swing1LimitDegrees=15, swing2LimitDegrees=10, twistLimitDegrees=5}
 {bone1Name=dyn_hair_03_side_l, bone2Name=dyn_hair_02_side_l, swing1LimitDegrees=20, swing2LimitDegrees=10, twistLimitDegrees=5}
 {bone1Name=dyn_hair_01_side_r, bone2Name=MSTR_hair_side_r, swing1LimitDegrees=15, swing2LimitDegrees=10, twistLimitDegrees=0}
 {bone1Name=dyn_hair_02_side_r, bone2Name=dyn_hair_01_side_r, swing1LimitDegrees=15, swing2LimitDegrees=15, twistLimitDegrees=5}
 {bone1Name=dyn_hair_03_side_r, bone2Name=dyn_hair_02_side_r, swing1LimitDegrees=20, swing2LimitDegrees=10, twistLimitDegrees=5}
]
