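# Physics asset for the SummitReedGrit male medium hat-variant face accessory, setting up secondary motion for hair and bandana.
# Four hair and three bandana chains of three capsules hang from small MSTR anchor spheres with limited swing and twist.
# Angularly free constraints cross-link chain tips, joining front to side hair and center to side bandana.
# Head, neck, upper spine, clavicle and upper-arm bodies provide body collision shapes.
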PhysicsAsset /BRCosmetics/Characters/Player/Male/Medium/Bodies/M_MED_SummitReedGrit_Hat/Meshes/Parts/M_MED_SummitReedGrit_FaceAccHat_Physics
bodies {
 MSTR_bandana [{shapeType=Sphere, radius=0.505, shapeName=MSTR_bandana_sphere}]
 MSTR_bandana_l [{shapeType=Sphere, radius=0.505, shapeName=MSTR_bandana_l_sphere}]
 MSTR_bandana_r [{shapeType=Sphere, radius=0.505, shapeName=MSTR_bandana_r_sphere}]
 MSTR_hair_fr_l [{shapeType=Sphere, radius=0.505, shapeName=MSTR_hair_fr_l_sphere}]
 MSTR_hair_fr_r [{shapeType=Sphere, radius=0.505, shapeName=MSTR_hair_fr_r_sphere}]
 MSTR_hair_side_l [{shapeType=Sphere, radius=0.505, shapeName=MSTR_hair_side_l_sphere}]
 MSTR_hair_side_r [{shapeType=Sphere, radius=0.505, shapeName=MSTR_hair_side_r_sphere}]
 clavicle_l [{
   shapeType=Capsule
   length=2.6125
   radius=6.8304
   shapeName=clavicle_l_capsule
   center=<9.8574 -1.6177 -0.98>
   rotation=<rot -0.0001 -90 97.3912>
  }]
 clavicle_r [{
   shapeType=Capsule
   length=2.6125
   radius=6.8304
   shapeName=clavicle_r_capsule
   center=<-9.9246 1.8533 0.4619>
   rotation=<rot -0.0001 -90 -82.6088>
  }]
 dyn_bandana_01 [{shapeType=Capsule, length=2.0211, radius=4.2389, shapeName=dyn_bandana_01_capsule, center=<3.7871 -2.3893 0>}]
 dyn_bandana_01_l [{
   shapeType=Capsule
   length=2.5818
   radius=3
   shapeName=dyn_bandana_01_l_capsule
   center=<4.2726 -3.5868 -2.3514>
   rotation=<rot 0.0002 0.0022 23.7183>
  }]
 dyn_bandana_01_r [{
   shapeType=Capsule
   length=2.5818
   radius=3
   shapeName=dyn_bandana_01_r_capsule
   center=<4.7391 -3.7572 2.2369>
   rotation=<rot -0.0002 0.0022 170.4133>
  }]
 dyn_bandana_02 [{shapeType=Capsule, length=1.7128, radius=5.1785, shapeName=dyn_bandana_02_capsule, center=<3.3503 -2.1055 0>}]
 dyn_bandana_02_l [{
   shapeType=Capsule
   length=2.8318
   radius=3.692
   shapeName=dyn_bandana_02_l_capsule
   center=<4.1435 -3.1894 -2.5328>
   rotation=<rot -2.7989 -1.0306 27.8144>
  }]
 dyn_bandana_02_r [{
   shapeType=Capsule
   length=2.748
   radius=3.7758
   shapeName=dyn_bandana_02_r_capsule
   center=<4.1559 -2.3126 2.6016>
   rotation=<rot 2.7989 -1.0305 -7.6115>
  }]
 dyn_bandana_03 [{shapeType=Capsule, length=2.8718, radius=4.2972, shapeName=dyn_bandana_03_capsule, center=<3.3503 -3.1242 0>}]
 dyn_bandana_03_l [{
   shapeType=Capsule
   length=3.6284
   radius=3.1407
   shapeName=dyn_bandana_03_l_capsule
   center=<4.3176 -3.2208 -2.4848>
   rotation=<rot -0.0002 0.0023 14.733>
  }]
 dyn_bandana_03_r [{
   shapeType=Capsule
   length=3.1482
   radius=3.621
   shapeName=dyn_bandana_03_r_capsule
   center=<4.3109 -3.4501 2.4592>
   rotation=<rot 3.425 -3.8122 -6.4701>
  }]
 dyn_hair_01_fr_l [{
   shapeType=Capsule
   length=2.2405
   radius=2.3803
   shapeName=dyn_hair_01_fr_l_capsule
   center=<2.4565 -1.9936 -1.2956>
   rotation=<rot 0 0 6.4028>
  }]
 dyn_hair_01_fr_r [{
   shapeType=Capsule
   length=2.2405
   radius=2.3803
   shapeName=dyn_hair_01_fr_r_capsule
   center=<-2.4565 1.9936 1.2956>
   rotation=<rot 0 0 -173.5972>
  }]
 dyn_hair_01_side_l [{
   shapeType=Capsule
   length=2.506
   radius=2.1148
   shapeName=dyn_hair_01_fr_l_capsule
   center=<2.4565 -1.7361 -1.6662>
   rotation=<rot 0.0001 0 -3.1943>
  }]
 dyn_hair_01_side_r [{
   shapeType=Capsule
   length=2.506
   radius=2.1148
   shapeName=dyn_hair_01_fr_r_capsule
   center=<-2.4565 1.7361 1.6662>
   rotation=<rot 0.0001 0 176.8057>
  }]
 dyn_hair_02_fr_l [{
   shapeType=Capsule
   length=2.2954
   radius=1.9818
   shapeName=dyn_hair_02_fr_l_capsule
   center=<1.981 -2.0937 -0.6803>
   rotation=<rot 0 0 14.3876>
  }]
 dyn_hair_02_fr_r [{
   shapeType=Capsule
   length=2.2954
   radius=1.9818
   shapeName=dyn_hair_02_fr_r_capsule
   center=<-1.981 2.0937 0.6803>
   rotation=<rot 0 0 -165.6124>
  }]
 dyn_hair_02_side_l [{
   shapeType=Capsule
   length=3.4753
   radius=2.2302
   shapeName=dyn_hair_02_fr_l_capsule
   center=<1.9893 -1.6343 -1.574>
   rotation=<rot -2.7146 0 0>
  }]
 dyn_hair_02_side_r [{
   shapeType=Capsule
   length=3.4753
   radius=2.2302
   shapeName=dyn_hair_02_fr_r_capsule
   center=<-1.9893 1.6343 1.574>
   rotation=<rot -2.7146 0 -180>
  }]
 dyn_hair_03_fr_l [{
   shapeType=Capsule
   length=2.1327
   radius=2.1881
   shapeName=dyn_hair_03_fr_l_capsule
   center=<0.9838 -2.0554 -0.4717>
   rotation=<rot 0 0 5.7997>
  }]
 dyn_hair_03_fr_r [{
   shapeType=Capsule
   length=2.1327
   radius=2.1881
   shapeName=dyn_hair_03_fr_r_capsule
   center=<-0.9838 2.0554 0.4717>
   rotation=<rot 0 0 -174.2003>
  }]
 dyn_hair_03_side_l [{
   shapeType=Capsule
   length=2.163
   radius=2.1577
   shapeName=dyn_hair_03_fr_l_capsule
   center=<0.8537 -1.7736 -1.622>
   rotation=<rot -6.7874 0 0>
  }]
 dyn_hair_03_side_r [{
   shapeType=Capsule
   length=2.163
   radius=2.1577
   shapeName=dyn_hair_03_fr_r_capsule
   center=<-0.8537 1.7736 1.622>
   rotation=<rot -6.7874 0 -180>
  }]
 head [{shapeType=Sphere, radius=7.03, shapeName=head_sphere, center=<0 1.903 0>}]
 neck_01 [{shapeType=Sphere, radius=5.6639, shapeName=neck_01_sphere, center=<0 0.1605 0>}]
 neck_02 [{shapeType=Sphere, radius=5.3844, shapeName=neck_01_sphere, center=<0 -0.7872 0>}]
 spine_05 [{shapeType=Capsule, length=10.0819, radius=9.6, shapeName=spine_05_capsule, center=<1.1226 -1.4292 0>, rotation=<rot 0 -90 0>}]
 upperarm_l [{
   shapeType=Capsule
   length=16.1125
   radius=4.8304
   shapeName=clavicle_l_capsule
   center=<9.9246 -1.3218 -0.4619>
   rotation=<rot -0.0002 -89.9999 87.3912>
  }]
 upperarm_r [{
   shapeType=Capsule
   length=16.1125
   radius=4.8304
   shapeName=clavicle_r_capsule
   center=<-9.9246 1.3218 0.4619>
   rotation=<rot -0.0002 -89.9999 -92.6088>
  }]
}
constraints [
 {bone1Name=dyn_hair_01_fr_l, bone2Name=MSTR_hair_fr_l, swing1LimitDegrees=15, swing2LimitDegrees=10, twistLimitDegrees=0}
 {bone1Name=dyn_hair_02_fr_l, bone2Name=dyn_hair_01_fr_l, swing1LimitDegrees=15, swing2LimitDegrees=10, twistLimitDegrees=5}
 {bone1Name=dyn_hair_03_fr_l, bone2Name=dyn_hair_02_fr_l, swing1LimitDegrees=20, swing2LimitDegrees=10, twistLimitDegrees=5}
 {
  bone1Name=dyn_hair_01_side_l
  bone2Name=MSTR_hair_side_l
  swing1LimitDegrees=15
  swing2LimitDegrees=10
  twistLimitDegrees=0
  twistMotion=Locked
 }
 {bone1Name=dyn_hair_02_side_l, bone2Name=dyn_hair_01_side_l, swing1LimitDegrees=15, swing2LimitDegrees=10, twistLimitDegrees=5}
 {bone1Name=dyn_hair_03_side_l, bone2Name=dyn_hair_02_side_l, swing1LimitDegrees=20, swing2LimitDegrees=10, twistLimitDegrees=5}
 {bone1Name=dyn_hair_01_fr_r, bone2Name=MSTR_hair_fr_r, swing1LimitDegrees=15, swing2LimitDegrees=10, twistLimitDegrees=0}
 {bone1Name=dyn_hair_02_fr_r, bone2Name=dyn_hair_01_fr_r, swing1LimitDegrees=15, swing2LimitDegrees=10, twistLimitDegrees=5}
 {bone1Name=dyn_hair_03_fr_r, bone2Name=dyn_hair_02_fr_r, swing1LimitDegrees=20, swing2LimitDegrees=10, twistLimitDegrees=5}
 {bone1Name=dyn_hair_01_side_r, bone2Name=MSTR_hair_side_r, swing1LimitDegrees=15, swing2LimitDegrees=10, twistLimitDegrees=0}
 {bone1Name=dyn_hair_02_side_r, bone2Name=dyn_hair_01_side_r, swing1LimitDegrees=15, swing2LimitDegrees=10, twistLimitDegrees=5}
 {bone1Name=dyn_hair_03_side_r, bone2Name=dyn_hair_02_side_r, swing1LimitDegrees=20, swing2LimitDegrees=10, twistLimitDegrees=5}
 {bone1Name=dyn_bandana_01_l, bone2Name=MSTR_bandana_l, swing1LimitDegrees=20, swing2LimitDegrees=20, twistLimitDegrees=5}
 {bone1Name=dyn_bandana_02_l, bone2Name=dyn_bandana_01_l, swing1LimitDegrees=20, swing2LimitDegrees=20, twistLimitDegrees=10}
 {bone1Name=dyn_bandana_03_l, bone2Name=dyn_bandana_02_l, swing1LimitDegrees=20, swing2LimitDegrees=20, twistLimitDegrees=10}
 {bone1Name=dyn_bandana_01, bone2Name=MSTR_bandana, swing1LimitDegrees=20, swing2LimitDegrees=20, twistLimitDegrees=5}
 {bone1Name=dyn_bandana_02, bone2Name=dyn_bandana_01, swing1LimitDegrees=20, swing2LimitDegrees=20, twistLimitDegrees=5}
 {bone1Name=dyn_bandana_03, bone2Name=dyn_bandana_02, swing1LimitDegrees=20, swing2LimitDegrees=20, twistLimitDegrees=5}
 {bone1Name=dyn_bandana_01_r, bone2Name=MSTR_bandana_r, swing1LimitDegrees=20, swing2LimitDegrees=20, twistLimitDegrees=5}
 {bone1Name=dyn_bandana_02_r, bone2Name=dyn_bandana_01_r, swing1LimitDegrees=20, swing2LimitDegrees=20, twistLimitDegrees=10}
 {bone1Name=dyn_bandana_03_r, bone2Name=dyn_bandana_02_r, swing1LimitDegrees=20, swing2LimitDegrees=20, twistLimitDegrees=10}
 {
  bone1Name=dyn_bandana_03
  bone2Name=dyn_bandana_03_r
  swing1LimitDegrees=45
  swing1Motion=Free
  swing2LimitDegrees=45
  swing2Motion=Free
  twistLimitDegrees=45
  twistMotion=Free
 }
 {
  bone1Name=dyn_bandana_03
  bone2Name=dyn_bandana_03_l
  swing1LimitDegrees=45
  swing1Motion=Free
  swing2LimitDegrees=45
  swing2Motion=Free
  twistLimitDegrees=45
  twistMotion=Free
 }
 {
  bone1Name=dyn_hair_03_fr_r
  bone2Name=dyn_hair_03_side_r
  swing1LimitDegrees=45
  swing1Motion=Free
  swing2LimitDegrees=45
  swing2Motion=Free
  twistLimitDegrees=15
  twistMotion=Free
 }
 {
  bone1Name=dyn_hair_03_fr_l
  bone2Name=dyn_hair_03_side_l
  swing1LimitDegrees=45
  swing1Motion=Free
  swing2LimitDegrees=45
  swing2Motion=Free
  twistLimitDegrees=15
  twistMotion=Free
 }
]
