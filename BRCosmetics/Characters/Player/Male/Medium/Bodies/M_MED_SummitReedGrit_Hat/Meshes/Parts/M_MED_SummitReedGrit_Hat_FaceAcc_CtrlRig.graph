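# Control Rig for the SummitReedGrit hat character's face accessories, procedurally adjusting bandana, hair, glasses and hat-cover bones.
# Parent constraints blend bandana and back-hair master bones between head, upper spine and clavicles.
# Head rotation relative to spine_05 is remapped into weights for additional corrective offsets, one also gated by low pelvis height.
# Forehead facial-bone movement shifts the glasses, with interpolated smoothing, and the hat cover.
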ControlRigBlueprint /BRCosmetics/Characters/Player/Male/Medium/Bodies/M_MED_SummitReedGrit_Hat/Meshes/Parts/M_MED_SummitReedGrit_Hat_FaceAcc_CtrlRig
graphs [
 RigVMFunctionLibrary {links [], nodes []}
 RigVMModel {nodes [
   RigUnit_BeginExecution = &RigUnit_BeginExecution()
   &RigUnit_ParentConstraint(
    AdvancedSettings=(InterpolationType=Average,RotationOrderForFilter=XZY)
    Child=(Type=Bone,Name="MSTR_bandana_main")
    Filter=(TranslationFilter=(bX=True,bY=false,bZ=True),RotationFilter=(bX=True,bY=True,bZ=True),ScaleFilter=(bX=True,bY=True,bZ=True))
    Parents=((Item=(Type=Bone,Name="head"),Weight=0.7),(Item=(Type=Bone,Name="spine_05"),Weight=0.3))
    Weight=1
    ExecutePin=@RigUnit_BeginExecution.ExecutePin
   )
   > &RigUnit_ParentConstraint(
    AdvancedSettings=(InterpolationType=Average,RotationOrderForFilter=XZY)
    Child=(Type=Bone,Name="MSTR_bandana_r")
    Filter=(TranslationFilter=(bX=true,bY=true,bZ=true),RotationFilter=(bX=true,bY=true,bZ=true),ScaleFilter=(bX=True,bY=True,bZ=True))
    Parents=((Item=(Type=Bone,Name="head"),Weight=0.5),(Item=(Type=Bone,Name="clavicle_r"),Weight=0.3),(Item=(Type=Bone,Name="spine_05"),Weight=0.8))
    Weight=1
   )
   > &RigUnit_ParentConstraint(
    AdvancedSettings=(InterpolationType=Average,RotationOrderForFilter=XZY)
    Child=(Type=Bone,Name="MSTR_bandana_l")
    Filter=(TranslationFilter=(bX=true,bY=true,bZ=true),RotationFilter=(bX=true,bY=true,bZ=true),ScaleFilter=(bX=True,bY=True,bZ=True))
    Parents=((Item=(Type=Bone,Name="head"),Weight=0.5),(Item=(Type=Bone,Name="clavicle_l"),Weight=0.3),(Item=(Type=Bone,Name="spine_05"),Weight=0.8))
    Weight=1
   )
   > &RigUnit_ParentConstraint(
    AdvancedSettings=(InterpolationType=Average,RotationOrderForFilter=XZY)
    Child=(Type=Bone,Name="MSTR_bandana")
    Filter=(TranslationFilter=(bX=true,bY=true,bZ=true),RotationFilter=(bX=true,bY=true,bZ=true),ScaleFilter=(bX=True,bY=True,bZ=True))
    Parents=((Item=(Type=Bone,Name="head"),Weight=0.5),(Item=(Type=Bone,Name="clavicle_l"),Weight=0.1),(Item=(Type=Bone,Name="spine_05"),Weight=0.8),(Item=(Type=Bone,Name="clavicle_r"),Weight=0.1))
    Weight=1
   )
   > &RigUnit_ParentConstraint(
    AdvancedSettings=(InterpolationType=Average,RotationOrderForFilter=XZY)
    Child=(Type=Bone,Name="MSTR_hair_bk")
    Filter=(TranslationFilter=(bX=true,bY=true,bZ=true),RotationFilter=(bX=True,bY=True,bZ=True),ScaleFilter=(bX=True,bY=True,bZ=True))
    Parents=((Item=(Type=Bone,Name="spine_05"),Weight=0.5),(Item=(Type=Bone,Name="head"),Weight=0.7))
    Weight=1
   )
   > &RigUnit_ParentConstraint(
    AdvancedSettings=(InterpolationType=Average,RotationOrderForFilter=XZY)
    Child=(Type=Bone,Name="MSTR_hair_bk_l")
    Filter=(TranslationFilter=(bX=true,bY=true,bZ=true),RotationFilter=(bX=True,bY=True,bZ=True),ScaleFilter=(bX=True,bY=True,bZ=True))
    Parents=((Item=(Type=Bone,Name="spine_05"),Weight=0.5),(Item=(Type=Bone,Name="head"),Weight=4))
    Weight=1
    bMaintainOffset=true
   )
   > &RigUnit_ParentConstraint(
    AdvancedSettings=(InterpolationType=Average,RotationOrderForFilter=XZY)
    Child=(Type=Bone,Name="MSTR_hair_bk_r")
    Filter=(TranslationFilter=(bX=true,bY=true,bZ=true),RotationFilter=(bX=True,bY=True,bZ=True),ScaleFilter=(bX=True,bY=True,bZ=True))
    Parents=((Item=(Type=Bone,Name="spine_05"),Weight=0.5),(Item=(Type=Bone,Name="head"),Weight=1))
    Weight=1
   )
   n1 = &RigVMFunction_MathQuaternionToEuler(
    RotationOrder=XYZ
    Value=&RigUnit_GetRelativeTransformForItem(Child=(Type=Bone,Name="head"), Parent=(Type=Bone,Name="spine_05"), bChildInitial=False, bParentInitial=False).RelativeTransform.Rotation
   )
   n2 = &RigVMFunction_MathDoubleRemap(SourceMaximum=-40, SourceMinimum=10, TargetMaximum=1, TargetMinimum=0, Value=@n1.Result.Z)
   > &RigUnit_ModifyTransforms(
    ItemToModify="((Item=(Type=Bone,Name=\"MSTR_hair_side_l\"),Transform=(Rotation=(X=-0.1722,Y=0.1285,Z=0.0227,W=0.9764),Translation=(X=-2,Y=0,Z=0),Scale3D=(X=0.95,Y=1,Z=1))),(Item=(Type=Bone,Name=\"MSTR_hair_fr_l\"),Transform=(Rotation=(X=0,Y=0.0872,Z=0,W=0.9962),Translation= ...(truncated)"
    WeightMinimum=0
    Weight=@n2.Result
   )
   > &RigUnit_ModifyTransforms(
    ItemToModify="((Item=(Type=Bone,Name=\"MSTR_hair_side_r\"),Transform=(Rotation=(X=-0.1722,Y=0.1285,Z=0.0227,W=0.9764),Translation=(X=2,Y=0,Z=0),Scale3D=(X=0.95,Y=1,Z=1))),(Item=(Type=Bone,Name=\"MSTR_hair_fr_r\"),Transform=(Rotation=(X=0,Y=0.0872,Z=0,W=0.9962),Translation=( ...(truncated)"
    WeightMinimum=0
    Weight=@n2.Result
   )
   n3 = &RigVMFunction_MathDoubleRemap(SourceMaximum=25, SourceMinimum=10, TargetMaximum=1, TargetMinimum=0, Value=@n1.Result.Z)
   > &RigUnit_ModifyTransforms(
    ItemToModify="((Item=(Type=Bone,Name=\"MSTR_hair_side_l\"),Transform=(Rotation=(X=0,Y=-0.2588,Z=0,W=0.9659),Translation=(X=0,Y=0,Z=0),Scale3D=(X=1,Y=1,Z=1))),(Item=(Type=Bone,Name=\"MSTR_hair_fr_l\"),Transform=(Rotation=(X=0.0226,Y=-0.2578,Z=0.0842,W=0.9623),Translation=(X= ...(truncated)"
    WeightMinimum=0
    Weight=@n3.Result
   )
   > &RigUnit_ModifyTransforms(
    ItemToModify="((Item=(Type=Bone,Name=\"MSTR_hair_side_r\"),Transform=(Rotation=(X=0,Y=-0.2588,Z=0,W=0.9659),Translation=(X=0,Y=0,Z=0),Scale3D=(X=1,Y=1,Z=1))),(Item=(Type=Bone,Name=\"MSTR_hair_fr_r\"),Transform=(Rotation=(X=0.0226,Y=-0.2578,Z=0.0842,W=0.9623),Translation=(X= ...(truncated)"
    WeightMinimum=0
    Weight=@n3.Result
   )
   n4 = &RigVMFunction_MathQuaternionToEuler(
    RotationOrder=ZYX
    Value=&RigUnit_GetRelativeTransformForItem(Child=(Type=Bone,Name="head"), Parent=(Type=Bone,Name="spine_05"), bChildInitial=False, bParentInitial=False).RelativeTransform.Rotation
   )
   > &RigUnit_ModifyTransforms(
    ItemToModify="((Item=(Type=Bone,Name=\"MSTR_bandana_main\"),Transform=(Rotation=(X=0,Y=0,Z=0,W=1),Translation=(X=-3,Y=-3,Z=0),Scale3D=(X=1,Y=1,Z=1))),(Item=(Type=Bone,Name=\"MSTR_hair_bk_l\"),Transform=(Rotation=(X=0.2083,Y=0.2192,Z=-0.3943,W=0.8678),Translation=(X=0,Y=2.9, ...(truncated)"
    WeightMinimum=0
    Weight=&RigVMFunction_MathDoubleRemap(SourceMaximum=-60, SourceMinimum=0, TargetMaximum=1, TargetMinimum=0, Value=@n4.Result.Z).Result
   )
   > &RigUnit_ModifyTransforms(
    ItemToModify=((Item=(Type=Bone,Name="MSTR_bandana_main"),Transform=(Rotation=(X=0,Y=0,Z=0.0436,W=0.999),Translation=(X=0,Y=0,Z=0),Scale3D=(X=1,Y=1,Z=1))))
    WeightMinimum=0
    Weight=&RigVMFunction_MathDoubleRemap(SourceMaximum=25, SourceMinimum=10, TargetMaximum=1, TargetMinimum=0, Value=@n4.Result.Z).Result
   )
   > &RigUnit_ModifyTransforms(
    ItemToModify="((Item=(Type=Bone,Name=\"MSTR_hair_bk_l\"),Transform=(Rotation=(X=0,Y=0,Z=0.0436,W=0.999),Translation=(X=0,Y=0.5,Z=-0.5),Scale3D=(X=1,Y=1,Z=1))),(Item=(Type=Bone,Name=\"MSTR_hair_bk_r\"),Transform=(Rotation=(X=0,Y=0,Z=0,W=1),Translation=(X=0,Y=-1,Z=1),Scale3D= ...(truncated)"
    Weight=1
    WeightMinimum=0
   )
   Remap = &RigVMFunction_MathDoubleRemap(
    SourceMaximum=70
    SourceMinimum=75
    TargetMaximum=1
    TargetMinimum=0
    Value=&RigUnit_GetRelativeTransformForItem(Child=(Type=Bone,Name="pelvis"), Parent=(Type=Bone,Name="root"), bChildInitial=False, bParentInitial=False).RelativeTransform.Translation.Z
   )
   Multiply = &RigVMFunction_MathDoubleMul(
    B=@Remap.Result
    A=&RigVMFunction_MathDoubleRemap(SourceMaximum=-90, SourceMinimum=-70, TargetMaximum=1, TargetMinimum=0, Value=@n4.Result.Z).Result
   )
   > &RigUnit_ModifyTransforms(
    ItemToModify="((Item=(Type=Bone,Name=\"MSTR_bandana_main\"),Transform=(Rotation=(X=0,Y=0,Z=-0.0436,W=0.999),Translation=(X=-3,Y=-4,Z=0),Scale3D=(X=1,Y=1,Z=1))),(Item=(Type=Bone,Name=\"dyn_knot\"),Transform=(Rotation=(X=0,Y=-0.3827,Z=0,W=0.9239),Translation=(X=0,Y=0,Z=3),Sca ...(truncated)"
    WeightMinimum=0
    Weight=@Multiply.Result
   )
   > &RigUnit_ModifyTransforms(
    ItemToModify="((Item=(Type=Bone,Name=\"MSTR_bandana_r\"),Transform=(Rotation=(X=0,Y=0,Z=-0.0872,W=0.9962),Translation=(X=0,Y=0,Z=0),Scale3D=(X=1,Y=1,Z=1))),(Item=(Type=Bone,Name=\"MSTR_bandana\"),Transform=(Rotation=(X=0,Y=0,Z=0,W=1),Translation=(X=0,Y=0,Z=0),Scale3D=(X=1,Y ...(truncated)"
    WeightMinimum=0
    Weight=&RigVMFunction_MathDoubleRemap(SourceMaximum=-45, SourceMinimum=0, TargetMaximum=1, TargetMinimum=0, Value=@n4.Result.X).Result
   )
   > &RigUnit_ModifyTransforms(
    ItemToModify="((Item=(Type=Bone,Name=\"MSTR_bandana_l\"),Transform=(Rotation=(X=0,Y=0,Z=-0.0872,W=0.9962),Translation=(X=0,Y=0,Z=0),Scale3D=(X=1,Y=1,Z=1))),(Item=(Type=Bone,Name=\"MSTR_bandana\"),Transform=(Rotation=(X=0,Y=0,Z=0,W=1),Translation=(X=0,Y=0,Z=0),Scale3D=(X=1,Y ...(truncated)"
    WeightMinimum=0
    Weight=&RigVMFunction_MathDoubleRemap(SourceMaximum=45, SourceMinimum=0, TargetMaximum=1, TargetMinimum=0, Value=@n4.Result.X).Result
   )
   n5 = &RigUnit_GetTransform(Item=(Type=Bone,Name="FACIAL_L_ForeheadIn"), Space=LocalSpace, bInitial=False)
   n6 = &RigUnit_GetTransform(Item=(Type=Bone,Name="FACIAL_R_ForeheadIn"), Space=LocalSpace, bInitial=False)
   Add = &RigVMFunction_MathDoubleAdd(
    A=&RigVMFunction_MathDoubleRemap(SourceMaximum=-2.8, SourceMinimum=-3, TargetMaximum=1, TargetMinimum=0, Value=@n5.Transform.Translation.Y).Result
    B=&RigVMFunction_MathDoubleRemap(SourceMaximum=-2.8, SourceMinimum=-3, TargetMaximum=1, TargetMinimum=0, Value=@n6.Transform.Translation.Y).Result
   )
   Clamp = &RigVMFunction_MathDoubleClamp(Value=@Add.Result)
   RigVMFunction_AlphaInterp = &RigVMFunction_AlphaInterp(InterpSpeedDecreasing=10, InterpSpeedIncreasing=10, Scale=1, bClampResult=False, bInterpResult=True, Value=@Clamp.Result)
   > &RigUnit_ModifyTransforms(
    ItemToModify=((Item=(Type=Bone,Name="def_glasses"),Transform=(Rotation=(X=0,Y=0,Z=0,W=1),Translation=(X=0.4,Y=0,Z=-0.1),Scale3D=(X=1,Y=1,Z=1))))
    WeightMinimum=0
    Weight=@RigVMFunction_AlphaInterp.Result
   )
   n7 = &RigVMFunction_MathDoubleRemap(SourceMaximum=-3.9, SourceMinimum=-3, TargetMaximum=1, TargetMinimum=0, Value=@n6.Transform.Translation.Y)
   n8 = &RigVMFunction_MathDoubleRemap(SourceMaximum=-3.9, SourceMinimum=-3, TargetMaximum=1, TargetMinimum=0, Value=@n5.Transform.Translation.Y)
   > &RigUnit_ModifyTransforms(
    ItemToModify=((Item=(Type=Bone,Name="def_hatcover"),Transform=(Rotation=(X=0,Y=0,Z=0,W=1),Translation=(X=1.5,Y=0,Z=0.2),Scale3D=(X=1,Y=1,Z=1))))
    WeightMinimum=0
    Weight=&RigVMFunction_MathDoubleClamp(Value=&RigVMFunction_MathDoubleAdd(B=@n7.Result, A=@n8.Result).Result).Result
   )
  ]}
]
hierarchy [
 root {loc=<0 0 0>, rot=<0 0 0 1>}
 pelvis {parent=root, loc=<0 -0.7071 93.1191>, rot=<0.0127 0.707 -0.0127 -0.707>}
 spine_01 {parent=pelvis, loc=<6.7494 -0.6388 0>, rot=<0 0 0.034 -0.9994>}
 spine_02 {parent=spine_01, loc=<7.2518 -0.035 0>, rot=<0 0 0.0238 -0.9997>}
 spine_03 {parent=spine_02, loc=<6.0104 0 0>, rot=<0 0 0.0149 -0.9999>}
 spine_04 {parent=spine_03, loc=<8.1024 0 0>, rot=<0 0 -0.0132 -0.9999>}
 spine_05 {parent=spine_04, loc=<10.0146 0 0>, rot=<0 0 -0.0015 -1>}
 neck_01 {parent=spine_05, loc=<11.3138 0.1529 0>, rot=<0 0 -0.2003 -0.9797>}
 neck_02 {parent=neck_01, loc=<4.5627 0 0>, rot=<0 0 0.0504 -0.9987>}
 head {parent=neck_02, loc=<5.3914 0 0>, rot=<0 0 0.115 -0.9934>}
 dyn_knot {parent=head, loc=<4.5292 -8.7624 -0.9059>, rot=<0.4555 -0.5408 0.451 -0.5446>}
 dyn_knot_up {parent=dyn_knot, loc=<0.4491 -0.198 0.7955>, rot=<-0.1511 0.16 -0.5785 -0.7854>}
 dyn_knot_down {parent=dyn_knot, loc=<0.6421 -0.062 -0.4195>, rot=<0.2064 -0.1602 -0.4465 -0.8558>}
 MSTR_bandana_main {parent=head, loc=<3.9355 -8.3379 0.9675>, rot=<0.1559 -0.9878 -0.0001 0.0005>}
 MSTR_bandana {parent=MSTR_bandana_main, loc=<1.7227 -0.0786 -0.0215>, rot=<0 0 0 1>}
 dyn_bandana_01 {parent=MSTR_bandana, loc=<0 0 0>, rot=<0 0 0 1>}
 dyn_bandana_02 {parent=dyn_bandana_01, loc=<8.5245 0 0>, rot=<-0.0024 0.0084 -0.0246 -0.9997>}
 dyn_bandana_03 {parent=dyn_bandana_02, loc=<7.4348 0 0>, rot=<-0.0003 0.0011 -0.0199 -0.9998>}
 dyn_bandana_end {parent=dyn_bandana_03, loc=<8.5563 0 0>, rot=<0 0 0 1>}
 MSTR_bandana_l {parent=MSTR_bandana_main, loc=<0.362 4.2025 8.0339>, rot=<0.046 0.1108 0.0457 -0.9917>}
 dyn_bandana_01_l {parent=MSTR_bandana_l, loc=<0 0 0>, rot=<0 0 0 1>}
 dyn_bandana_02_l {parent=dyn_bandana_01_l, loc=<8.8775 0 0>, rot=<-0.0784 -0.0225 0.0126 -0.9966>}
 dyn_bandana_03_l {parent=dyn_bandana_02_l, loc=<7.6998 0 0>, rot=<0.0077 -0.0174 0.0084 -0.9998>}
 dyn_bandana_end_l {parent=dyn_bandana_03_l, loc=<8.2893 0 0>, rot=<0 0 0 1>}
 MSTR_bandana_r {parent=MSTR_bandana_main, loc=<0.4016 4.4546 -5.947>, rot=<-0.1263 -0.1081 0.0812 -0.9827>}
 dyn_bandana_01_r {parent=MSTR_bandana_r, loc=<0 0 0>, rot=<0 0 0 1>}
 dyn_bandana_02_r {parent=dyn_bandana_01_r, loc=<9.5635 0 0>, rot=<-0.0081 0.01 -0.0276 -0.9995>}
 dyn_bandana_03_r {parent=dyn_bandana_02_r, loc=<7.6765 0 0>, rot=<0.0776 0.0215 -0.0131 -0.9967>}
 dyn_bandana_end_r {parent=dyn_bandana_03_r, loc=<7.4213 0 0>, rot=<0 0 0 1>}
 MSTR_hair_fr_l {parent=head, loc=<4.0837 4.857 -6.2257>, rot=<0.0401 -0.7067 -0.706 0.0204>}
 dyn_hair_01_fr_l {parent=MSTR_hair_fr_l, loc=<0 0 0>, rot=<0 0 0 1>}
 dyn_hair_02_fr_l {parent=dyn_hair_01_fr_l, loc=<5.1821 0 0>, rot=<0.0021 0.0014 -0.0272 -0.9996>}
 dyn_hair_03_fr_l {parent=dyn_hair_02_fr_l, loc=<5.4763 0 0>, rot=<0 0 0 1>}
 MSTR_hair_side_l {parent=head, loc=<4.2936 0.7988 -6.8112>, rot=<-0.0322 0.7065 0.7066 0.0227>}
 dyn_hair_01_side_l {parent=MSTR_hair_side_l, loc=<0 0 0>, rot=<0 0 0 1>}
 dyn_hair_02_side_l {parent=dyn_hair_01_side_l, loc=<5.1605 0 0>, rot=<0.0002 0.0026 -0.0633 -0.998>}
 dyn_hair_03_side_l {parent=dyn_hair_02_side_l, loc=<5.3365 0 0>, rot=<0 0 0 1>}
 MSTR_hair_bk_l {parent=head, loc=<4.2215 -4.5293 -5.7339>, rot=<0.0458 -0.9057 -0.4214 -0.0075>}
 dyn_hair_01_bk_l {parent=MSTR_hair_bk_l, loc=<0 0 0>, rot=<0 0 0 1>}
 dyn_hair_02_bk_l {parent=dyn_hair_01_bk_l, loc=<5.1428 0 0>, rot=<-0.0003 0.0036 0.003 -1>}
 dyn_hair_03_bk_l {parent=dyn_hair_02_bk_l, loc=<4.3533 0 0>, rot=<0 0 0 1>}
 MSTR_hair_bk {parent=head, loc=<4.4845 -7.2133 -0.1486>, rot=<0.0873 -0.9962 0 0>}
 dyn_hair_01_bk {parent=MSTR_hair_bk, loc=<0 0 0>, rot=<0 0 0 1>}
 dyn_hair_02_bk {parent=dyn_hair_01_bk, loc=<5.0351 0 0>, rot=<0 0 -0.0619 -0.9981>}
 dyn_hair_03_bk {parent=dyn_hair_02_bk, loc=<4.7301 0 0>, rot=<0 0 0 1>}
 MSTR_hair_fr_r {parent=head, loc=<4.0838 4.857 6.2257>, rot=<-0.7067 -0.0401 -0.0204 -0.706>}
 dyn_hair_01_fr_r {parent=MSTR_hair_fr_r, loc=<0 0 0>, rot=<0 0 0 1>}
 dyn_hair_02_fr_r {parent=dyn_hair_01_fr_r, loc=<-5.1825 0 0>, rot=<0.0021 0.0014 -0.0272 -0.9996>}
 dyn_hair_03_fr_r {parent=dyn_hair_02_fr_r, loc=<-5.4762 0 0>, rot=<0 0 0 1>}
 MSTR_hair_side_r {parent=head, loc=<4.2933 0.7988 6.8112>, rot=<-0.7065 -0.0322 0.0227 -0.7066>}
 dyn_hair_01_side_r {parent=MSTR_hair_side_r, loc=<0 0 0>, rot=<0 0 0 1>}
 dyn_hair_02_side_r {parent=dyn_hair_01_side_r, loc=<-5.1606 0 0>, rot=<0.0002 0.0026 -0.0633 -0.998>}
 dyn_hair_03_side_r {parent=dyn_hair_02_side_r, loc=<-5.3364 0 0>, rot=<0 0 0 1>}
 MSTR_hair_bk_r {parent=head, loc=<4.2214 -4.5293 5.7339>, rot=<0.9057 0.0458 -0.0075 0.4214>}
 dyn_hair_01_bk_r {parent=MSTR_hair_bk_r, loc=<0 0 0>, rot=<0 0 0 1>}
 dyn_hair_02_bk_r {parent=dyn_hair_01_bk_r, loc=<-5.1431 0 0>, rot=<-0.0003 0.0036 0.003 -1>}
 dyn_hair_03_bk_r {parent=dyn_hair_02_bk_r, loc=<-4.353 0 0>, rot=<0 0 0 -1>}
 FACIAL_C_FacialRoot {parent=head, loc=<2.152 -0.746 0>, rot=<0.003 -0.7238 -0.69 0.0029>}
 FACIAL_C_LowerLipRotation {parent=FACIAL_C_FacialRoot, loc=<0.7099 0.136 2.7514>, rot=<0.239 -0.2071 0.6761 -0.6655>}
 FACIAL_C_MouthLower {parent=FACIAL_C_LowerLipRotation, loc=<0.0041 -1.0513 5.9732>, rot=<-0.342 0 0 -0.9397>}
 FACIAL_C_LipLower {parent=FACIAL_C_MouthLower, loc=<0.0024 1.6381 2.2679>, rot=<0.0872 0 0 -0.9962>}
 FACIAL_L_LipLower {parent=FACIAL_C_MouthLower, loc=<0.9466 1.4812 1.997>, rot=<0.0927 -0.1294 0.0272 -0.9869>}
 FACIAL_R_LipLower {parent=FACIAL_C_MouthLower, loc=<-0.9472 1.4843 2.0054>, rot=<0.0927 0.1294 -0.0272 -0.9869>}
 FACIAL_L_LipLowerOuter {parent=FACIAL_C_MouthLower, loc=<1.635 0.96 1.6771>, rot=<0.0913 -0.2321 0.0318 -0.9679>}
 FACIAL_R_LipLowerOuter {parent=FACIAL_C_MouthLower, loc=<-1.6391 0.9611 1.6835>, rot=<0.0913 0.2321 -0.0318 -0.9679>}
 FACIAL_C_TeethLower {parent=FACIAL_C_LowerLipRotation, loc=<0.003 -1.6302 7.0465>, rot=<-0.342 0 0 -0.9397>}
 FACIAL_C_Tongue1 {parent=FACIAL_C_TeethLower, loc=<0.0024 -0.0759 -5.2082>, rot=<-0.2588 0 0 -0.9659>}
 FACIAL_C_Tongue2 {parent=FACIAL_C_Tongue1, loc=<-0.0003 0.4736 1.4654>, rot=<0.2588 0 0 -0.9659>}
 FACIAL_C_Tongue3 {parent=FACIAL_C_Tongue2, loc=<-0.0005 0.2021 1.4104>, rot=<0.1305 0 0 -0.9914>}
 FACIAL_C_Tongue4 {parent=FACIAL_C_Tongue3, loc=<0.0002 -0.1684 1.151>, rot=<0.0436 0 0 -0.999>}
 FACIAL_C_TongueUpper3 {parent=FACIAL_C_Tongue3, loc=<-0.0004 -0.3898 -0.1005>, rot=<0 0 0 1>}
 FACIAL_C_TongueLower3 {parent=FACIAL_C_Tongue3, loc=<0.0003 0.7854 0.1203>, rot=<0 0 0 1>}
 FACIAL_L_TongueSide3 {parent=FACIAL_C_Tongue3, loc=<1.545 0.0512 0.2077>, rot=<0 0 0 1>}
 FACIAL_R_TongueSide3 {parent=FACIAL_C_Tongue3, loc=<-1.5452 0.0523 0.2091>, rot=<0 0 0 1>}
 FACIAL_C_TongueUpper2 {parent=FACIAL_C_Tongue2, loc=<-0.0005 -0.533 0.1395>, rot=<0 0 0 1>}
 FACIAL_L_TongueSide2 {parent=FACIAL_C_Tongue2, loc=<1.8262 0.1897 0.524>, rot=<0 0 0 1>}
 FACIAL_R_TongueSide2 {parent=FACIAL_C_Tongue2, loc=<-1.8257 0.1903 0.5199>, rot=<0 0 0 1>}
 FACIAL_C_Jaw {parent=FACIAL_C_FacialRoot, loc=<0.7099 0.136 2.7514>, rot=<0.239 -0.2071 0.6761 -0.6655>}
 FACIAL_C_Jawline {parent=FACIAL_C_Jaw, loc=<0.0039 2.3926 8.5301>, rot=<0.2586 0 0 -0.966>}
 FACIAL_L_ChinSide {parent=FACIAL_C_Jaw, loc=<2.4594 0.5713 7.7925>, rot=<-0.2208 -0.3738 0.0499 -0.8995>}
 FACIAL_R_ChinSide {parent=FACIAL_C_Jaw, loc=<-2.4464 0.569 7.7934>, rot=<-0.2208 0.3738 -0.0499 -0.8995>}
 FACIAL_C_Chin {parent=FACIAL_C_Jaw, loc=<0.004 0.959 8.3844>, rot=<-0.2591 0 0 -0.9659>}
 FACIAL_L_CheekInner {parent=FACIAL_C_FacialRoot, loc=<-0.1339 -1.4008 8.7459>, rot=<-0.145 0.0609 0.5398 -0.827>}
 FACIAL_R_CheekInner {parent=FACIAL_C_FacialRoot, loc=<-0.1269 2.2285 8.5696>, rot=<-0.0352 0.1842 0.8191 -0.5421>}
 FACIAL_R_CheekOuter {parent=FACIAL_C_FacialRoot, loc=<0.0332 3.7893 6.9122>, rot=<0.3061 0.2616 0.5812 -0.7071>}
 FACIAL_L_CheekOuter {parent=FACIAL_C_FacialRoot, loc=<0.0224 -3.1122 7.2431>, rot=<-0.2336 -0.2721 0.7209 -0.5931>}
 FACIAL_C_Nose {parent=FACIAL_C_FacialRoot, loc=<-0.7961 0.4861 10.1299>, rot=<-0.0952 0.1284 0.6953 -0.7007>}
 FACIAL_C_NoseLower {parent=FACIAL_C_Nose, loc=<0.0011 2.7766 -0.1111>, rot=<0.1305 0 0 -0.9914>}
 FACIAL_L_Nostril {parent=FACIAL_C_Nose, loc=<1.3529 2.2703 0.0558>, rot=<0.0173 0.1305 0.0023 -0.9913>}
 FACIAL_R_Nostril {parent=FACIAL_C_Nose, loc=<-1.3511 2.2425 0.0337>, rot=<0.0173 -0.1305 -0.0023 -0.9913>}
 FACIAL_C_NoseTip {parent=FACIAL_C_Nose, loc=<-0.0012 1.8115 0.683>, rot=<0.0872 0 0 -0.9962>}
 FACIAL_C_MouthUpper {parent=FACIAL_C_FacialRoot, loc=<3.4447 0.3899 8.1588>, rot=<-0.003 0.0367 0.7061 -0.7071>}
 FACIAL_C_LipUpper {parent=FACIAL_C_MouthUpper, loc=<-0.0013 -0.2823 2.6775>, rot=<-0.0872 0 0 -0.9962>}
 FACIAL_L_LipUpper {parent=FACIAL_C_MouthUpper, loc=<1.408 -0.1222 2.3057>, rot=<-0.0778 -0.1601 -0.0379 -0.9833>}
 FACIAL_R_LipUpper {parent=FACIAL_C_MouthUpper, loc=<-1.4097 -0.1187 2.3085>, rot=<-0.0778 0.1601 0.0379 -0.9833>}
 FACIAL_L_LipUpperOuter {parent=FACIAL_C_MouthUpper, loc=<1.8752 0.0874 1.8209>, rot=<-0.0984 -0.2293 -0.0983 -0.9634>}
 FACIAL_R_LipUpperOuter {parent=FACIAL_C_MouthUpper, loc=<-1.8738 0.1004 1.8266>, rot=<-0.0984 0.2293 0.0983 -0.9634>}
 FACIAL_L_LipCorner {parent=FACIAL_C_MouthUpper, loc=<2.0073 0.4382 1.2891>, rot=<0 -0.2078 0 -0.9782>}
 FACIAL_R_LipCorner {parent=FACIAL_C_MouthUpper, loc=<-2.0648 0.432 1.3194>, rot=<0 0.2078 0 -0.9782>}
 FACIAL_L_ForeheadIn {parent=FACIAL_C_FacialRoot, loc=<-4.0314 -1.016 9.5464>, rot=<-0.0952 0.0302 0.7037 -0.7035>}
 FACIAL_R_ForeheadIn {parent=FACIAL_C_FacialRoot, loc=<-4.0322 1.9233 9.4054>, rot=<0.0034 0.1286 0.6981 -0.7044>}
 def_hatcover {parent=head, loc=<6.4526 9.2829 -0.0457>, rot=<0.4979 0.5021 0.5021 0.4979>}
 def_glasses {parent=head, loc=<4.962 -1.2204 0>, rot=<-0.5045 -0.4955 -0.4955 -0.5045>}
 clavicle_l {parent=spine_05, loc=<4.3575 -0.1466 -2.3223>, rot=<0.0465 -0.7289 -0.009 -0.6829>}
 upperarm_l {parent=clavicle_l, loc=<15.5111 0 0>, rot=<0.0226 -0.3234 -0.0088 -0.9459>}
 clavicle_r {parent=spine_05, loc=<4.3575 -0.1466 2.3223>, rot=<0.7289 0.0465 -0.6829 0.009>}
 upperarm_r {parent=clavicle_r, loc=<-15.5111 0 0>, rot=<0.0226 -0.3234 -0.0088 -0.9459>}
]
variables [isFrontend {type=bool, default=False}]
