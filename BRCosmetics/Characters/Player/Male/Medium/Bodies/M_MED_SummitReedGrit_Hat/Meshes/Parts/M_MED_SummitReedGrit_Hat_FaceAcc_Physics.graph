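# Physics asset for the SummitReedGrit Hat character's face accessory part, covering secondary motion of a bandana and hair.
# Three bandana and four hair chains each run three capsules from a small MSTR anchor sphere, with mostly tight swing and twist limits.
# Free-angle constraints cross-link the bandana tips and each side's front and side hair tips.
# Head, neck, upper spine, clavicle and upper-arm bodies have no constraints, apparently serving as colliders.
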PhysicsAsset /BRCosmetics/Characters/Player/Male/Medium/Bodies/M_MED_SummitReedGrit_Hat/Meshes/Parts/M_MED_SummitReedGrit_Hat_FaceAcc_Physics
bodies {
 MSTR_bandana [{shapeType=Sphere, radius=0.505, shapeName=MSTR_bandana_sphere}]
 MSTR_bandana_l [{shapeType=Sphere, radius=0.505, shapeName=MSTR_bandana_l_sphere}]
 MSTR_bandana_r [{shapeType=Sphere, radius=0.505, shapeName=MSTR_bandana_r_sphere}]
 MSTR_hair_fr_l [{shapeType=Sphere, radius=0.505, shapeName=MSTR_hair_fr_l_sphere}]
 MSTR_hair_fr_r [{shapeType=Sphere, radius=0.505, shapeName=MSTR_hair_fr_r_sphere}]
 MSTR_hair_side_l [{shapeType=Sphere, radius=0.505, shapeName=MSTR_hair_side_l_sphere}]
 MSTR_hair_side_r [{shapeType=Sphere, radius=0.505, shapeName=MSTR_hair_side_r_sphere}]
 clavicle_l [{
   shapeType=Capsule
   length=2.6125
   radius=6.8304
   shapeName=clavicle_l_capsule
   center=<9.8574 -2.115 -0.98>
   rotation=<rot -0.0001 -90 97.3912>
  }]
 clavicle_r [{
   shapeType=Capsule
   length=3.0929
   radius=7.6585
   shapeName=clavicle_r_capsule
   center=<-7.853 1.6859 1.1799>
   rotation=<rot -0.3035 -87.6615 -90.4235>
  }]
 dyn_bandana_01 [{shapeType=Capsule, length=2.0212, radius=4.2389, shapeName=dyn_bandana_01_capsule, center=<3.7871 -1.6651 0>}]
 dyn_bandana_01_l [{
   shapeType=Capsule
   length=2.5818
   radius=3
   shapeName=dyn_bandana_01_l_capsule
   center=<4.2726 -3.3472 -2.4884>
   rotation=<rot 0.0002 0.0022 31.7261>
  }]
 dyn_bandana_01_r [{
   shapeType=Capsule
   length=2.5818
   radius=3
   shapeName=dyn_bandana_01_r_capsule
   center=<4.7391 -3.1029 2.5665>
   rotation=<rot -0.0002 0.0022 153.263>
  }]
 dyn_bandana_02 [{shapeType=Capsule, length=4.341, radius=6.2055, shapeName=dyn_bandana_02_capsule, center=<3.3503 -3.4077 0>}]
 dyn_bandana_02_l [{
   shapeType=Capsule
   length=3.6823
   radius=3.692
   shapeName=dyn_bandana_02_l_capsule
   center=<4.1885 -3.617 -1.5644>
   rotation=<rot 4.1719 179.1406 156.5689>
  }]
 dyn_bandana_02_r [{
   shapeType=Capsule
   length=4.2086
   radius=3.7758
   shapeName=dyn_bandana_02_r_capsule
   center=<4.1321 -2.8313 2.1105>
   rotation=<rot 7.8397 -2.8845 -3.6945>
  }]
 dyn_bandana_03 [{
   shapeType=Capsule
   length=7.0449
   radius=6.2445
   shapeName=dyn_bandana_03_capsule
   center=<3.9207 -4.2164 0.097>
   rotation=<rot -1.4215 0.1193 1.1912>
  }]
 dyn_bandana_03_l [{
   shapeType=Capsule
   length=4.7011
   radius=3.621
   shapeName=dyn_bandana_03_r_capsule
   center=<6.4789 -2.9913 -2.0469>
   rotation=<rot -4.3174 8.5396 20.3938>
  }]
 dyn_bandana_03_r [{
   shapeType=Capsule
   length=4.1637
   radius=3.621
   shapeName=dyn_bandana_03_r_capsule
   center=<5.2059 -2.3846 2.3062>
   rotation=<rot 3.4041 8.2237 -2.4325>
  }]
 dyn_hair_01_fr_l [{
   shapeType=Capsule
   length=2.2405
   radius=2.3803
   shapeName=dyn_hair_01_fr_l_capsule
   center=<2.4565 -1.9936 -1.2956>
   rotation=<rot 0 0 6.4028>
  }]
 dyn_hair_01_fr_r [{
   shapeType=Capsule
   length=2.2405
   radius=2.3803
   shapeName=dyn_hair_01_fr_r_capsule
   center=<-2.4565 1.9936 1.2956>
   rotation=<rot 0 0 -173.5972>
  }]
 dyn_hair_01_side_l [{
   shapeType=Capsule
   length=2.506
   radius=2.1148
   shapeName=dyn_hair_01_fr_l_capsule
   center=<2.4565 -1.7361 -1.6662>
   rotation=<rot 0.0001 0 -3.1943>
  }]
 dyn_hair_01_side_r [{
   shapeType=Capsule
   length=2.506
   radius=2.1148
   shapeName=dyn_hair_01_fr_r_capsule
   center=<-2.4565 1.7361 1.6662>
   rotation=<rot 0.0001 0 176.8057>
  }]
 dyn_hair_02_fr_l [{
   shapeType=Capsule
   length=2.2954
   radius=1.9818
   shapeName=dyn_hair_02_fr_l_capsule
   center=<1.981 -2.0937 -0.6803>
   rotation=<rot 0 0 14.3876>
  }]
 dyn_hair_02_fr_r [{
   shapeType=Capsule
   length=2.2954
   radius=1.9818
   shapeName=dyn_hair_02_fr_r_capsule
   center=<-1.981 2.0937 0.6803>
   rotation=<rot 0 0 -165.6124>
  }]
 dyn_hair_02_side_l [{
   shapeType=Capsule
   length=3.4753
   radius=2.2302
   shapeName=dyn_hair_02_fr_l_capsule
   center=<1.9893 -1.6343 -1.574>
   rotation=<rot -2.7146 0 0>
  }]
 dyn_hair_02_side_r [{
   shapeType=Capsule
   length=3.4753
   radius=2.2302
   shapeName=dyn_hair_02_fr_r_capsule
   center=<-1.9893 1.6343 1.574>
   rotation=<rot -2.7146 0 -180>
  }]
 dyn_hair_03_fr_l [{
   shapeType=Capsule
   length=2.1327
   radius=2.1881
   shapeName=dyn_hair_03_fr_l_capsule
   center=<0.9838 -2.0553 -0.4717>
   rotation=<rot 0 0 5.7997>
  }]
 dyn_hair_03_fr_r [{
   shapeType=Capsule
   length=2.9008
   radius=1.4199
   shapeName=dyn_hair_03_fr_r_capsule
   center=<-0.9838 2.0553 0.4717>
   rotation=<rot 0 0 -174.2003>
  }]
 dyn_hair_03_side_l [{
   shapeType=Capsule
   length=2.163
   radius=2.1577
   shapeName=dyn_hair_03_fr_l_capsule
   center=<0.8537 -1.7736 -1.622>
   rotation=<rot -6.7874 0 0>
  }]
 dyn_hair_03_side_r [{
   shapeType=Capsule
   length=2.163
   radius=2.1577
   shapeName=dyn_hair_03_fr_r_capsule
   center=<-0.8537 1.7736 1.622>
   rotation=<rot -6.7874 0 -180>
  }]
 head [{shapeType=Sphere, radius=7.03, shapeName=head_sphere, center=<0 1.903 0>}]
 neck_01 [{shapeType=Sphere, radius=5.6639, shapeName=neck_01_sphere, center=<0 0.1605 0>}]
 neck_02 [{shapeType=Sphere, radius=5.3844, shapeName=neck_01_sphere, center=<0 -0.7872 0>}]
 spine_05 [
  {shapeType=Capsule, length=9.1168, radius=9.8943, shapeName=spine_05_capsule, center=<-1.1843 -1.3904 0>, rotation=<rot 0 -90 0>}
  {shapeType=Capsule, length=12.9106, radius=6.5852, shapeName=spine_05_capsule, center=<6.7445 -2.0377 0>, rotation=<rot 0 7.2086 0>}
 ]
 upperarm_l [{
   shapeType=Capsule
   length=23.7721
   radius=4.8304
   shapeName=clavicle_l_capsule
   center=<9.9246 -1.3218 -0.4619>
   rotation=<rot -0.0002 -89.9999 87.3912>
  }]
 upperarm_r [{
   shapeType=Capsule
   length=23.6045
   radius=5.4254
   shapeName=clavicle_r_capsule
   center=<-9.9418 1.3218 0.8388>
   rotation=<rot -0.0002 -89.9999 -92.6088>
  }]
}
constraints [
 {bone1Name=dyn_bandana_01, bone2Name=MSTR_bandana, swing1LimitDegrees=16, swing2LimitDegrees=16, twistLimitDegrees=5}
 {bone1Name=dyn_bandana_02, bone2Name=dyn_bandana_01, swing1LimitDegrees=16, swing2LimitDegrees=10, twistLimitDegrees=5, twistMotion=Locked}
 {bone1Name=dyn_bandana_03, bone2Name=dyn_bandana_02, swing1LimitDegrees=20, swing2LimitDegrees=20, twistLimitDegrees=5}
 {
  bone1Name=dyn_bandana_03
  bone2Name=dyn_bandana_03_r
  swing1LimitDegrees=45
  swing1Motion=Free
  swing2LimitDegrees=45
  swing2Motion=Free
  twistLimitDegrees=45
  twistMotion=Free
 }
 {
  bone1Name=dyn_bandana_03
  bone2Name=dyn_bandana_03_l
  swing1LimitDegrees=45
  swing1Motion=Free
  swing2LimitDegrees=45
  swing2Motion=Free
  twistLimitDegrees=45
  twistMotion=Free
 }
 {bone1Name=dyn_bandana_01_l, bone2Name=MSTR_bandana_l, swing1LimitDegrees=16, swing2LimitDegrees=10, twistLimitDegrees=5}
 {
  bone1Name=dyn_bandana_02_l
  bone2Name=dyn_bandana_01_l
  swing1LimitDegrees=16
  swing2LimitDegrees=10
  twistLimitDegrees=10
  twistMotion=Locked
 }
 {
  bone1Name=dyn_bandana_03_l
  bone2Name=dyn_bandana_02_l
  swing1LimitDegrees=20
  swing2LimitDegrees=10
  twistLimitDegrees=10
  twistMotion=Locked
 }
 {bone1Name=dyn_bandana_01_r, bone2Name=MSTR_bandana_r, swing1LimitDegrees=16, swing2LimitDegrees=10, twistLimitDegrees=5}
 {
  bone1Name=dyn_bandana_02_r
  bone2Name=dyn_bandana_01_r
  swing1LimitDegrees=16
  swing2LimitDegrees=10
  twistLimitDegrees=10
  twistMotion=Locked
 }
 {
  bone1Name=dyn_bandana_03_r
  bone2Name=dyn_bandana_02_r
  swing1LimitDegrees=20
  swing2LimitDegrees=10
  twistLimitDegrees=0
  twistMotion=Locked
 }
 {bone1Name=dyn_hair_01_fr_l, bone2Name=MSTR_hair_fr_l, swing1LimitDegrees=15, swing2LimitDegrees=10, twistLimitDegrees=0}
 {bone1Name=dyn_hair_02_fr_l, bone2Name=dyn_hair_01_fr_l, swing1LimitDegrees=15, swing2LimitDegrees=10, twistLimitDegrees=5}
 {bone1Name=dyn_hair_03_fr_l, bone2Name=dyn_hair_02_fr_l, swing1LimitDegrees=20, swing2LimitDegrees=10, twistLimitDegrees=5}
 {
  bone1Name=dyn_hair_03_fr_l
  bone2Name=dyn_hair_03_side_l
  swing1LimitDegrees=45
  swing1Motion=Free
  swing2LimitDegrees=45
  swing2Motion=Free
  twistLimitDegrees=15
  twistMotion=Free
 }
 {bone1Name=dyn_hair_01_fr_r, bone2Name=MSTR_hair_fr_r, swing1LimitDegrees=15, swing2LimitDegrees=10, twistLimitDegrees=0}
 {bone1Name=dyn_hair_02_fr_r, bone2Name=dyn_hair_01_fr_r, swing1LimitDegrees=15, swing2LimitDegrees=10, twistLimitDegrees=5}
 {bone1Name=dyn_hair_03_fr_r, bone2Name=dyn_hair_02_fr_r, swing1LimitDegrees=20, swing2LimitDegrees=10, twistLimitDegrees=5}
 {
  bone1Name=dyn_hair_03_fr_r
  bone2Name=dyn_hair_03_side_r
  swing1LimitDegrees=45
  swing1Motion=Free
  swing2LimitDegrees=45
  swing2Motion=Free
  twistLimitDegrees=15
  twistMotion=Free
 }
 {
  bone1Name=dyn_hair_01_side_l
  bone2Name=MSTR_hair_side_l
  swing1LimitDegrees=15
  swing2LimitDegrees=10
  twistLimitDegrees=0
  twistMotion=Locked
 }
 {bone1Name=dyn_hair_02_side_l, bone2Name=dyn_hair_01_side_l, swing1LimitDegrees=15, swing2LimitDegrees=10, twistLimitDegrees=5}
 {bone1Name=dyn_hair_03_side_l, bone2Name=dyn_hair_02_side_l, swing1LimitDegrees=20, swing2LimitDegrees=10, twistLimitDegrees=5}
 {bone1Name=dyn_hair_01_side_r, bone2Name=MSTR_hair_side_r, swing1LimitDegrees=15, swing2LimitDegrees=10, twistLimitDegrees=0}
 {bone1Name=dyn_hair_02_side_r, bone2Name=dyn_hair_01_side_r, swing1LimitDegrees=15, swing2LimitDegrees=10, twistLimitDegrees=5}
 {bone1Name=dyn_hair_03_side_r, bone2Name=dyn_hair_02_side_r, swing1LimitDegrees=20, swing2LimitDegrees=10, twistLimitDegrees=5}
]
