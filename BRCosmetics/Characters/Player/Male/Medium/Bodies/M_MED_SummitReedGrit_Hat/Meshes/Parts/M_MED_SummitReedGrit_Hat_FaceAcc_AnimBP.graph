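# Post-process style AnimBP for the SummitReedGrit hat face accessory, adding secondary motion to its bandana, hair and knot bones.
# A Control Rig corrects the pose copied from the parent mesh, then feeds RigidBody simulation, with a separate skydive physics asset, when enabled without backbling.
# Otherwise AnimDynamics chains drive bandana and hair, with backpack-dependent bandana setups; knot dynamics always run last.
# Each update computes per-movement-state gravity and RigidBody settings.
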AnimBlueprint /BRCosmetics/Characters/Player/Male/Medium/Bodies/M_MED_SummitReedGrit_Hat/Meshes/Parts/M_MED_SummitReedGrit_Hat_FaceAcc_AnimBP
target_skeleton=&/M_MED_SummitReedGrit_Hat_FaceAcc_Skeleton
graphs [
 AnimGraph {nodes [
   n1 = "Transform (Modify) Bone - Bone: MSTR_bandana_main"(ComponentPose=Local To Component(LocalPose=Use cached pose 'DeformConstr'()), Alpha=1, Translation="0,2,0", Rotation="0,0,0", Scale="1,1,1") {
    rotationMode=BMM_Additive
    rotationSpace=BCS_BoneSpace
    translationMode=BMM_Additive
    translationSpace=BCS_BoneSpace
    alphaBoolBlend {blendInTime=0.05, blendOutTime=0.05}
    alphaScaleBiasClamp {bInterpResult=true}
    boneToModify {boneName=MSTR_bandana_main}
    translation {y=2}
   }
   n2 = RigidBody!(
    ComponentPose=@n1
    Alpha
    OverrideWorldGravity:Vector="0,0,-980"
    ExternalForce
    SimSpaceSettings:Sim Space Settings Structure=(WorldAlpha=1,VelocityScaleZ=1,DampingAlpha=0,MaxLinearVelocity=10,MaxAngularVelocity=15,MaxLinearAcceleration=1100,MaxAngularAcceleration=150,ExternalLinearDragV=(X=0,Y=0,Z=0),ExternalLinearVelocity=(X=0,Y=0,Z=0),ExternalAngularVelocity=(X=0,Y=0,Z=0))
   ) -> (Pose) {
    bClampLinearTranslationLimitToRefPose=true
    bOverrideWorldGravity=true
    overridePhysicsAsset=&/M_MED_SummitReedGrit_Hat_FaceAcc_Physics
    simulationSpace=BaseBoneSpace
    baseBoneRef {boneName=spine_05}
    simSpaceSettings {dampingAlpha=0, maxAngularAcceleration=150, maxAngularVelocity=15, maxLinearAcceleration=1100, maxLinearVelocity=10, worldAlpha=1}
   }
   n3 = RigidBody(
    ComponentPose=Local To Component(LocalPose=Use cached pose 'DeformConstr'())
    OverrideWorldGravity="0,0,-980"
    SimSpaceSettings=(WorldAlpha=1,VelocityScaleZ=1,DampingAlpha=0,MaxLinearVelocity=10,MaxAngularVelocity=15,MaxLinearAcceleration=1100,MaxAngularAcceleration=150,ExternalLinearDragV=(X=0,Y=0,Z=0),ExternalLinearVelocity=(X=0,Y=0,Z=0),ExternalAngularVelocity=(X=0,Y=0,Z=0))
   ) {
    bClampLinearTranslationLimitToRefPose=true
    bOverrideWorldGravity=true
    overridePhysicsAsset=&/M_MED_SummitReedGrit_Hat_FaceAcc_SkyDive_Physics
    simulationSpace=BaseBoneSpace
    baseBoneRef {boneName=spine_05}
    simSpaceSettings {dampingAlpha=0, maxAngularAcceleration=150, maxAngularVelocity=15, maxLinearAcceleration=1100, maxLinearVelocity=10, worldAlpha=1}
   }
   n4 = Blend Poses by bool(BlendPose_0=Component To Local(ComponentPose=@n3), BlendPose_1=Component To Local(ComponentPose=@n2), bActiveValue=Get bIsSkyDiving())
   n5 = Anim Dynamics - Bone: dyn_hair_01_fr_l(ComponentPose=Local To Component(LocalPose=Use cached pose 'BandanaAnimDynamics'()), Alpha=1, ExternalForce="0,0,0") {
    angularDampingOverride=0.999
    angularSpringConstant=600
    bAngularSpring=true
    bGravityOverrideInSimSpace=true
    bOverrideAngularDamping=true
    bUseGravityOverride=true
    simulationSpace=RootRelative
    boundBone {boneName=dyn_hair_01_fr_l}
    componentAppliedLinearAccClamp=<1500 1500 2000>
    componentLinearAccScale=<0.3 0.3 0.5>
    componentLinearVelScale=<0.3 0.3 0.2>
    gravityOverride {z=-980}
    physicsBodyDefinitions [{
      boundBone {boneName=dyn_hair_01_fr_l}
      boxExtents=<12 12 12>
      constraintSetup {angularLimitsMax=<0 5 0>, angularLimitsMin=<0 -10 -7>, angularTarget=<0 0 0>}
      localJointOffset=<1 0 0>
     }]
    relativeSpaceBone {boneName=head}
    simSpaceSettings {maxAngularAcceleration=1500, maxAngularVelocity=1500, simSpaceAngularAlpha=0.1}
   }
   n6 = Anim Dynamics - Bone: dyn_hair_02_fr_l(ComponentPose=@n5, Alpha=1, ExternalForce="0,0,0") {
    angularDampingOverride=0.999
    angularSpringConstant=600
    bAngularSpring=true
    bGravityOverrideInSimSpace=true
    bOverrideAngularDamping=true
    bUseGravityOverride=true
    simulationSpace=RootRelative
    boundBone {boneName=dyn_hair_02_fr_l}
    componentAppliedLinearAccClamp=<1500 1500 2000>
    componentLinearAccScale=<0.3 0.3 0.5>
    componentLinearVelScale=<0.3 0.3 0.2>
    gravityOverride {z=-980}
    physicsBodyDefinitions [{
      boundBone {boneName=dyn_hair_02_fr_l}
      boxExtents=<12 12 12>
      constraintSetup {angularLimitsMax=<0 5 10>, angularLimitsMin=<0 -10 -10>, angularTarget=<0 0 0>}
      localJointOffset=<1 0 0>
     }]
    relativeSpaceBone {boneName=head}
    simSpaceSettings {maxAngularAcceleration=1500, maxAngularVelocity=1500, simSpaceAngularAlpha=0.1}
   }
   n7 = Anim Dynamics - Bone: dyn_hair_03_fr_l(ComponentPose=@n6, Alpha=1, ExternalForce="0,0,0") {
    angularDampingOverride=0.999
    angularSpringConstant=600
    bAngularSpring=true
    bGravityOverrideInSimSpace=true
    bOverrideAngularDamping=true
    bUseGravityOverride=true
    simulationSpace=RootRelative
    boundBone {boneName=dyn_hair_03_fr_l}
    componentAppliedLinearAccClamp=<1500 1500 2000>
    componentLinearAccScale=<0.3 0.3 0.5>
    componentLinearVelScale=<0.3 0.3 0.2>
    gravityOverride {z=-980}
    physicsBodyDefinitions [{
      boundBone {boneName=dyn_hair_03_fr_l}
      boxExtents=<12 12 12>
      constraintSetup {angularLimitsMax=<0 5 0>, angularLimitsMin=<0 -10 -10>, angularTarget=<0 0 0>}
      localJointOffset=<1 0 0>
     }]
    relativeSpaceBone {boneName=head}
    simSpaceSettings {maxAngularAcceleration=1500, maxAngularVelocity=1500, simSpaceAngularAlpha=0.1}
   }
   n8 = Anim Dynamics - Bone: dyn_hair_01_fr_r(ComponentPose=@n7, Alpha=1, ExternalForce="0,0,0") {
    angularDampingOverride=0.999
    angularSpringConstant=600
    bAngularSpring=true
    bGravityOverrideInSimSpace=true
    bOverrideAngularDamping=true
    bUseGravityOverride=true
    simulationSpace=RootRelative
    boundBone {boneName=dyn_hair_01_fr_r}
    componentAppliedLinearAccClamp=<1500 1500 2000>
    componentLinearAccScale=<0.3 0.3 0.5>
    componentLinearVelScale=<0.3 0.3 0.2>
    gravityOverride {z=-980}
    physicsBodyDefinitions [{
      boundBone {boneName=dyn_hair_01_fr_r}
      boxExtents=<12 12 12>
      constraintSetup {angularLimitsMax=<0 5 0>, angularLimitsMin=<0 -10 -7>, angularTarget=<0 0 0>}
      localJointOffset=<-1 0 0>
     }]
    relativeSpaceBone {boneName=head}
    simSpaceSettings {maxAngularAcceleration=1500, maxAngularVelocity=1500, simSpaceAngularAlpha=0.1}
   }
   n9 = Anim Dynamics - Bone: dyn_hair_02_fr_r(ComponentPose=@n8, Alpha=1, ExternalForce="0,0,0") {
    angularDampingOverride=0.999
    angularSpringConstant=600
    bAngularSpring=true
    bGravityOverrideInSimSpace=true
    bOverrideAngularDamping=true
    bUseGravityOverride=true
    simulationSpace=RootRelative
    boundBone {boneName=dyn_hair_02_fr_r}
    componentAppliedLinearAccClamp=<1500 1500 2000>
    componentLinearAccScale=<0.3 0.3 0.5>
    componentLinearVelScale=<0.3 0.3 0.2>
    gravityOverride {z=-980}
    physicsBodyDefinitions [{
      boundBone {boneName=dyn_hair_02_fr_r}
      boxExtents=<12 12 12>
      constraintSetup {angularLimitsMax=<0 5 10>, angularLimitsMin=<0 -10 -10>, angularTarget=<0 0 0>}
      localJointOffset=<-1 0 0>
     }]
    relativeSpaceBone {boneName=head}
    simSpaceSettings {maxAngularAcceleration=1500, maxAngularVelocity=1500, simSpaceAngularAlpha=0.1}
   }
   n10 = Anim Dynamics - Bone: dyn_hair_03_fr_r(ComponentPose=@n9, Alpha=1, ExternalForce="0,0,0") {
    angularDampingOverride=0.999
    angularSpringConstant=600
    bAngularSpring=true
    bGravityOverrideInSimSpace=true
    bOverrideAngularDamping=true
    bUseGravityOverride=true
    simulationSpace=RootRelative
    boundBone {boneName=dyn_hair_03_fr_r}
    componentAppliedLinearAccClamp=<1500 1500 2000>
    componentLinearAccScale=<0.3 0.3 0.5>
    componentLinearVelScale=<0.3 0.3 0.2>
    gravityOverride {z=-980}
    physicsBodyDefinitions [{
      boundBone {boneName=dyn_hair_03_fr_r}
      boxExtents=<12 12 12>
      constraintSetup {angularLimitsMax=<0 5 0>, angularLimitsMin=<0 -10 -10>, angularTarget=<0 0 0>}
      localJointOffset=<-1 0 0>
     }]
    relativeSpaceBone {boneName=head}
    simSpaceSettings {maxAngularAcceleration=1500, maxAngularVelocity=1500, simSpaceAngularAlpha=0.1}
   }
   n11 = Anim Dynamics - Bone: dyn_hair_01_side_l(ComponentPose=@n10, Alpha=1, ExternalForce="0,0,0") {
    angularDampingOverride=0.999
    angularSpringConstant=600
    bAngularSpring=true
    bGravityOverrideInSimSpace=true
    bOverrideAngularDamping=true
    bUseGravityOverride=true
    simulationSpace=RootRelative
    boundBone {boneName=dyn_hair_01_side_l}
    componentAppliedLinearAccClamp=<1500 1500 2000>
    componentLinearAccScale=<0.3 0.3 0.5>
    componentLinearVelScale=<0.3 0.3 0.2>
    gravityOverride {z=-980}
    physicsBodyDefinitions [{
      boundBone {boneName=dyn_hair_01_side_l}
      boxExtents=<12 12 12>
      constraintSetup {angularLimitsMax=<0 5 0>, angularLimitsMin=<0 -10 -7>, angularTarget=<0 0 0>}
      localJointOffset=<1 0 0>
     }]
    relativeSpaceBone {boneName=head}
    simSpaceSettings {maxAngularAcceleration=1500, maxAngularVelocity=1500, simSpaceAngularAlpha=0.1}
   }
   n12 = Anim Dynamics - Bone: dyn_hair_02_side_l(ComponentPose=@n11, Alpha=1, ExternalForce="0,0,0") {
    angularDampingOverride=0.999
    angularSpringConstant=600
    bAngularSpring=true
    bGravityOverrideInSimSpace=true
    bOverrideAngularDamping=true
    bUseGravityOverride=true
    simulationSpace=RootRelative
    boundBone {boneName=dyn_hair_02_side_l}
    componentAppliedLinearAccClamp=<1500 1500 2000>
    componentLinearAccScale=<0.3 0.3 0.5>
    componentLinearVelScale=<0.3 0.3 0.2>
    gravityOverride {z=-980}
    physicsBodyDefinitions [{
      boundBone {boneName=dyn_hair_02_side_l}
      boxExtents=<12 12 12>
      constraintSetup {angularLimitsMax=<0 5 10>, angularLimitsMin=<0 -10 -7>, angularTarget=<0 0 0>}
      localJointOffset=<1 0 0>
     }]
    relativeSpaceBone {boneName=head}
    simSpaceSettings {maxAngularAcceleration=1500, maxAngularVelocity=1500, simSpaceAngularAlpha=0.1}
   }
   n13 = Anim Dynamics - Bone: dyn_hair_03_side_l(ComponentPose=@n12, Alpha=1, ExternalForce="0,0,0") {
    angularDampingOverride=0.999
    angularSpringConstant=600
    bAngularSpring=true
    bGravityOverrideInSimSpace=true
    bOverrideAngularDamping=true
    bUseGravityOverride=true
    simulationSpace=RootRelative
    boundBone {boneName=dyn_hair_03_side_l}
    componentAppliedLinearAccClamp=<1500 1500 2000>
    componentLinearAccScale=<0.3 0.3 0.5>
    componentLinearVelScale=<0.3 0.3 0.2>
    gravityOverride {z=-980}
    physicsBodyDefinitions [{
      boundBone {boneName=dyn_hair_03_side_l}
      boxExtents=<12 12 12>
      constraintSetup {angularLimitsMax=<0 5 0>, angularLimitsMin=<0 -10 -7>, angularTarget=<0 0 0>}
      localJointOffset=<1 0 0>
     }]
    relativeSpaceBone {boneName=head}
    simSpaceSettings {maxAngularAcceleration=1500, maxAngularVelocity=1500, simSpaceAngularAlpha=0.1}
   }
   n14 = Anim Dynamics - Bone: dyn_hair_01_side_r(ComponentPose=@n13, Alpha=1, ExternalForce="0,0,0") {
    angularDampingOverride=0.999
    angularSpringConstant=600
    bAngularSpring=true
    bGravityOverrideInSimSpace=true
    bOverrideAngularDamping=true
    bUseGravityOverride=true
    simulationSpace=RootRelative
    boundBone {boneName=dyn_hair_01_side_r}
    componentAppliedLinearAccClamp=<1500 1500 2000>
    componentLinearAccScale=<0.3 0.3 0.5>
    componentLinearVelScale=<0.3 0.3 0.2>
    gravityOverride {z=-980}
    physicsBodyDefinitions [{
      boundBone {boneName=dyn_hair_01_side_r}
      boxExtents=<12 12 12>
      constraintSetup {angularLimitsMax=<0 5 0>, angularLimitsMin=<0 -10 -7>, angularTarget=<0 0 0>}
      localJointOffset=<-1 0 0>
     }]
    relativeSpaceBone {boneName=head}
    simSpaceSettings {maxAngularAcceleration=1500, maxAngularVelocity=1500, simSpaceAngularAlpha=0.1}
   }
   n15 = Anim Dynamics - Bone: dyn_hair_02_side_r(ComponentPose=@n14, Alpha=1, ExternalForce="0,0,0") {
    angularDampingOverride=0.999
    angularSpringConstant=600
    bAngularSpring=true
    bGravityOverrideInSimSpace=true
    bOverrideAngularDamping=true
    bUseGravityOverride=true
    simulationSpace=RootRelative
    boundBone {boneName=dyn_hair_02_side_r}
    componentAppliedLinearAccClamp=<1500 1500 2000>
    componentLinearAccScale=<0.3 0.3 0.5>
    componentLinearVelScale=<0.3 0.3 0.2>
    gravityOverride {z=-980}
    physicsBodyDefinitions [{
      boundBone {boneName=dyn_hair_02_side_r}
      boxExtents=<12 12 12>
      constraintSetup {angularLimitsMax=<0 5 10>, angularLimitsMin=<0 -10 -10>, angularTarget=<0 0 0>}
      localJointOffset=<-1 0 0>
     }]
    relativeSpaceBone {boneName=head}
    simSpaceSettings {maxAngularAcceleration=1500, maxAngularVelocity=1500, simSpaceAngularAlpha=0.1}
   }
   n16 = Anim Dynamics - Bone: dyn_hair_03_side_r(ComponentPose=@n15, Alpha=1, ExternalForce="0,0,0") {
    angularDampingOverride=0.999
    angularSpringConstant=600
    bAngularSpring=true
    bGravityOverrideInSimSpace=true
    bOverrideAngularDamping=true
    bUseGravityOverride=true
    simulationSpace=RootRelative
    boundBone {boneName=dyn_hair_03_side_r}
    componentAppliedLinearAccClamp=<1500 1500 2000>
    componentLinearAccScale=<0.3 0.3 0.5>
    componentLinearVelScale=<0.3 0.3 0.2>
    gravityOverride {z=-980}
    physicsBodyDefinitions [{
      boundBone {boneName=dyn_hair_03_side_r}
      boxExtents=<12 12 12>
      constraintSetup {angularLimitsMax=<0 5 0>, angularLimitsMin=<0 -10 -10>, angularTarget=<0 0 0>}
      localJointOffset=<-1 0 0>
     }]
    relativeSpaceBone {boneName=head}
    simSpaceSettings {maxAngularAcceleration=1500, maxAngularVelocity=1500, simSpaceAngularAlpha=0.1}
   }
   n17 = Component To Local(ComponentPose=@n16)
   n18 = Get IsRigidBodyEnableBackPackNotEquip() -> (IsRigidBodyEnableBackPackNotEquip:bool)
   n19 = Anim Dynamics - Bone: dyn_knot(
    ComponentPose=Local To Component(LocalPose=Blend Poses by bool(BlendPose_0=@n4, BlendPose_1=@n17, bActiveValue=@n18))
    Alpha=1
    ExternalForce="0,0,0"
   ) {
    angularDampingOverride=0.99
    angularSpringConstant=600
    bAngularSpring=true
    bOverrideAngularDamping=true
    bUseGravityOverride=true
    simulationSpace=RootRelative
    boundBone {boneName=dyn_knot}
    componentAppliedLinearAccClamp=<1500 1500 2000>
    componentLinearAccScale=<0.3 0.3 0.5>
    componentLinearVelScale=<0.3 0.3 0.3>
    gravityOverride {z=-980}
    physicsBodyDefinitions [{
      boundBone {boneName=dyn_knot}
      boxExtents=<10 10 10>
      constraintSetup {angularLimitsMax=<0 10 10>, angularLimitsMin=<0 -25 -10>, angularTarget=<1 0 0>}
      localJointOffset=<0.5 0 0>
     }]
    simSpaceSettings {maxAngularAcceleration=1500, maxAngularVelocity=1500, simSpaceAngularAlpha=0.1}
   }
   n20 = Anim Dynamics - Bone: dyn_knot_up(ComponentPose=@n19, Alpha=1, ExternalForce="0,0,0") {
    angularDampingOverride=0.99
    angularSpringConstant=600
    bAngularSpring=true
    bOverrideAngularDamping=true
    bUseGravityOverride=true
    simulationSpace=RootRelative
    boundBone {boneName=dyn_knot_up}
    componentAppliedLinearAccClamp=<1500 1500 2000>
    componentLinearAccScale=<0.3 0.3 0.5>
    componentLinearVelScale=<0.3 0.3 0.3>
    gravityOverride {z=-980}
    physicsBodyDefinitions [{
      boundBone {boneName=dyn_knot_up}
      boxExtents=<10 10 10>
      constraintSetup {angularLimitsMax=<0 10 30>, angularLimitsMin=<0 -25 -10>, angularTarget=<1 0 0>}
      localJointOffset=<0.5 0 0>
     }]
    simSpaceSettings {maxAngularAcceleration=1500, maxAngularVelocity=1500, simSpaceAngularAlpha=0.1}
   }
   n21 = Anim Dynamics - Bone: dyn_knot_down(ComponentPose=@n20, Alpha=1, ExternalForce="0,0,0") {
    angularDampingOverride=0.99
    angularSpringConstant=600
    bAngularSpring=true
    bOverrideAngularDamping=true
    bUseGravityOverride=true
    simulationSpace=RootRelative
    boundBone {boneName=dyn_knot_down}
    componentAppliedLinearAccClamp=<1500 1500 2000>
    componentLinearAccScale=<0.3 0.3 0.5>
    componentLinearVelScale=<0.3 0.3 0.3>
    gravityOverride {z=-980}
    physicsBodyDefinitions [{
      boundBone {boneName=dyn_knot_down}
      boxExtents=<10 10 10>
      constraintSetup {angularLimitsMax=<0 0 45>, angularLimitsMin=<0 -25 -10>, angularTarget=<1 0 0>}
      localJointOffset=<0.5 0 0>
     }]
    simSpaceSettings {maxAngularAcceleration=1500, maxAngularVelocity=1500, simSpaceAngularAlpha=0.1}
   }
   Output Pose(Result=Component To Local(ComponentPose=@n21)) {layerGroup=None}
   n22 = Control Rig(Source=Copy Pose From Mesh() {bUseAttachedParent=true}) {controlRigAssetReference {blueprintRigClass=&/M_MED_SummitReedGrit_Hat_FaceAcc_CtrlRig.M_MED_SummitReedGrit_Hat_FaceAcc_CtrlRig_C}}
   Save cached pose 'DeformConstr'(Pose=@n22) {cacheName=DeformConstr}
   n23 = Anim Dynamics - Bone: dyn_bandana_03(ComponentPose=Local To Component(LocalPose=Use cached pose 'DeformConstr'()), Alpha=1, ExternalForce="0,0,0") {
    angularDampingOverride=0.99
    angularSpringConstant=100
    bAngularSpring=true
    bOverrideAngularDamping=true
    bUseGravityOverride=true
    simulationSpace=RootRelative
    boundBone {boneName=dyn_bandana_03}
    componentAppliedLinearAccClamp=<1500 1500 2000>
    componentLinearAccScale=<0.3 0.3 0.5>
    componentLinearVelScale=<0.3 0.3 0.3>
    gravityOverride {z=-980}
    physicsBodyDefinitions [{
      boundBone {boneName=dyn_bandana_03}
      boxExtents=<10 10 10>
      constraintSetup {angularLimitsMax=<0 15 0>, angularLimitsMin=<0 -15 -5>, angularTarget=<1 0 0>}
      localJointOffset=<1 0 0>
     }]
    simSpaceSettings {maxAngularAcceleration=1500, maxAngularVelocity=1500, simSpaceAngularAlpha=0.1}
   }
   n24 = Anim Dynamics - Bone: dyn_bandana_03_l(ComponentPose=@n23, Alpha=1, ExternalForce="0,0,0") {
    angularDampingOverride=0.99
    angularSpringConstant=100
    bAngularSpring=true
    bOverrideAngularDamping=true
    bUseGravityOverride=true
    simulationSpace=RootRelative
    boundBone {boneName=dyn_bandana_03_l}
    componentAppliedLinearAccClamp=<1500 1500 2000>
    componentLinearAccScale=<0.3 0.3 0.5>
    componentLinearVelScale=<0.3 0.3 0.3>
    gravityOverride {z=-980}
    physicsBodyDefinitions [{
      boundBone {boneName=dyn_bandana_03_l}
      boxExtents=<10 10 10>
      constraintSetup {angularLimitsMax=<0 15 0>, angularLimitsMin=<0 -15 -5>, angularTarget=<1 0 0>}
      localJointOffset=<1 0 0>
     }]
    simSpaceSettings {maxAngularAcceleration=1500, maxAngularVelocity=1500, simSpaceAngularAlpha=0.1}
   }
   n25 = Anim Dynamics - Bone: dyn_bandana_03_r(ComponentPose=@n24, Alpha=1, ExternalForce="0,0,0") {
    angularDampingOverride=0.99
    angularSpringConstant=100
    bAngularSpring=true
    bOverrideAngularDamping=true
    bUseGravityOverride=true
    simulationSpace=RootRelative
    boundBone {boneName=dyn_bandana_03_r}
    componentAppliedLinearAccClamp=<1500 1500 2000>
    componentLinearAccScale=<0.3 0.3 0.5>
    componentLinearVelScale=<0.3 0.3 0.3>
    gravityOverride {z=-980}
    physicsBodyDefinitions [{
      boundBone {boneName=dyn_bandana_03_r}
      boxExtents=<10 10 10>
      constraintSetup {angularLimitsMax=<0 15 0>, angularLimitsMin=<0 -15 -5>, angularTarget=<1 0 0>}
      localJointOffset=<1 0 0>
     }]
    simSpaceSettings {maxAngularAcceleration=1500, maxAngularVelocity=1500, simSpaceAngularAlpha=0.1}
   }
   n26 = "Transform (Modify) Bone - Bone: MSTR_bandana_l"!(
    ComponentPose=Local To Component(LocalPose=Use cached pose 'DeformConstr'())
    bAlphaBoolEnabled:bool=True
    Translation="0,0,0"
    Rotation="-5,-5,0"
    Scale="1,1,1"
   ) -> (Pose) {
    alphaInputType=Bool
    rotationMode=BMM_Additive
    rotationSpace=BCS_BoneSpace
    alphaBoolBlend {blendInTime=0.1, blendOption=Linear, blendOutTime=0.1}
    boneToModify {boneName=MSTR_bandana_l}
    rotation {pitch=-5, yaw=-5}
   }
   n27 = "Transform (Modify) Bone - Bone: MSTR_bandana_r"!(ComponentPose=@n26, bAlphaBoolEnabled:bool=True, Translation="0,0,0", Rotation="5,-5,0", Scale="1,1,1") -> (Pose) {
    alphaInputType=Bool
    rotationMode=BMM_Additive
    rotationSpace=BCS_BoneSpace
    alphaBoolBlend {blendInTime=0.1, blendOption=Linear, blendOutTime=0.1}
    boneToModify {boneName=MSTR_bandana_r}
    rotation {pitch=5, yaw=-5}
   }
   n28 = "Transform (Modify) Bone - Bone: MSTR_bandana"!(ComponentPose=@n27, bAlphaBoolEnabled:bool=True, Translation="0,0,0", Rotation="0,-5,0", Scale="1,1,1") -> (Pose) {
    alphaInputType=Bool
    rotationMode=BMM_Additive
    rotationSpace=BCS_BoneSpace
    alphaBoolBlend {blendInTime=0.1, blendOption=Linear, blendOutTime=0.1}
    boneToModify {boneName=MSTR_bandana}
    rotation {yaw=-5}
   }
   n29 = Anim Dynamics - Bone: dyn_bandana_01(ComponentPose=@n28, Alpha=1, ExternalForce="0,0,0") {
    angularDampingOverride=0.99
    angularSpringConstant=600
    bAngularSpring=true
    bOverrideAngularDamping=true
    bUseGravityOverride=true
    simulationSpace=RootRelative
    boundBone {boneName=dyn_bandana_01}
    componentAppliedLinearAccClamp=<1500 1500 2000>
    componentLinearAccScale=<0.3 0.3 0.5>
    componentLinearVelScale=<0.3 0.3 0.2>
    gravityOverride {z=-980}
    physicsBodyDefinitions [{
      boundBone {boneName=dyn_bandana_01}
      boxExtents=<10 10 10>
      constraintSetup {angularLimitsMax=<0 0 0>, angularLimitsMin=<0 0 -20>, angularTarget=<0 0 0>}
      localJointOffset=<1 0 0>
     }]
    simSpaceSettings {maxAngularAcceleration=1500, maxAngularVelocity=1500, simSpaceAngularAlpha=0.1}
   }
   n30 = Anim Dynamics - Bone: dyn_bandana_02(ComponentPose=@n29, Alpha=1, ExternalForce="0,0,0") {
    angularDampingOverride=0.99
    angularSpringConstant=600
    bAngularSpring=true
    bOverrideAngularDamping=true
    bUseGravityOverride=true
    simulationSpace=RootRelative
    boundBone {boneName=dyn_bandana_02}
    componentAppliedLinearAccClamp=<1500 1500 2000>
    componentLinearAccScale=<0.3 0.3 0.5>
    componentLinearVelScale=<0.3 0.3 0.2>
    gravityOverride {z=-980}
    physicsBodyDefinitions [{
      boundBone {boneName=dyn_bandana_02}
      boxExtents=<10 10 10>
      constraintSetup {angularLimitsMax=<0 0 0>, angularLimitsMin=<0 0 -30>, angularTarget=<0 0 0>}
      localJointOffset=<1 0 0>
     }]
    simSpaceSettings {maxAngularAcceleration=1500, maxAngularVelocity=1500, simSpaceAngularAlpha=0.1}
   }
   n31 = Anim Dynamics - Bone: dyn_bandana_03(ComponentPose=@n30, Alpha=1, ExternalForce="0,0,0") {
    angularDampingOverride=0.99
    angularSpringConstant=600
    bAngularSpring=true
    bOverrideAngularDamping=true
    bUseGravityOverride=true
    simulationSpace=RootRelative
    boundBone {boneName=dyn_bandana_03}
    componentAppliedLinearAccClamp=<1500 1500 2000>
    componentLinearAccScale=<0.3 0.3 0.5>
    componentLinearVelScale=<0.3 0.3 0.3>
    gravityOverride {z=-980}
    physicsBodyDefinitions [{
      boundBone {boneName=dyn_bandana_03}
      boxExtents=<10 10 10>
      constraintSetup {angularLimitsMax=<0 0 0>, angularLimitsMin=<0 0 -45>, angularTarget=<0 0 0>}
      localJointOffset=<1 0 0>
     }]
    simSpaceSettings {maxAngularAcceleration=1500, maxAngularVelocity=1500, simSpaceAngularAlpha=0.1}
   }
   n32 = Anim Dynamics - Bone: dyn_bandana_01_l(ComponentPose=@n31, Alpha=1, ExternalForce="0,0,0") {
    angularDampingOverride=0.99
    angularSpringConstant=600
    bAngularSpring=true
    bOverrideAngularDamping=true
    bUseGravityOverride=true
    simulationSpace=RootRelative
    boundBone {boneName=dyn_bandana_01_l}
    componentAppliedLinearAccClamp=<1500 1500 2000>
    componentLinearAccScale=<0.3 0.3 0.5>
    componentLinearVelScale=<0.3 0.3 0.2>
    gravityOverride {z=-980}
    physicsBodyDefinitions [{
      boundBone {boneName=dyn_bandana_01_l}
      boxExtents=<10 10 10>
      constraintSetup {angularLimitsMax=<0 0 0>, angularLimitsMin=<0 0 -20>, angularTarget=<0 0 0>}
      localJointOffset=<1 0 0>
     }]
    simSpaceSettings {maxAngularAcceleration=1500, maxAngularVelocity=1500, simSpaceAngularAlpha=0.1}
   }
   n33 = Anim Dynamics - Bone: dyn_bandana_02_l(ComponentPose=@n32, Alpha=1, ExternalForce="0,0,0") {
    angularDampingOverride=0.99
    angularSpringConstant=600
    bAngularSpring=true
    bOverrideAngularDamping=true
    bUseGravityOverride=true
    simulationSpace=RootRelative
    boundBone {boneName=dyn_bandana_02_l}
    componentAppliedLinearAccClamp=<1500 1500 2000>
    componentLinearAccScale=<0.3 0.3 0.5>
    componentLinearVelScale=<0.3 0.3 0.2>
    gravityOverride {z=-980}
    physicsBodyDefinitions [{
      boundBone {boneName=dyn_bandana_02_l}
      boxExtents=<10 10 10>
      constraintSetup {angularLimitsMax=<0 0 0>, angularLimitsMin=<0 0 -30>, angularTarget=<0 0 0>}
      localJointOffset=<1 0 0>
     }]
    simSpaceSettings {maxAngularAcceleration=1500, maxAngularVelocity=1500, simSpaceAngularAlpha=0.1}
   }
   n34 = Anim Dynamics - Bone: dyn_bandana_03_l(ComponentPose=@n33, Alpha=1, ExternalForce="0,0,0") {
    angularDampingOverride=0.99
    angularSpringConstant=600
    bAngularSpring=true
    bOverrideAngularDamping=true
    bUseGravityOverride=true
    simulationSpace=RootRelative
    boundBone {boneName=dyn_bandana_03_l}
    componentAppliedLinearAccClamp=<1500 1500 2000>
    componentLinearAccScale=<0.3 0.3 0.5>
    componentLinearVelScale=<0.3 0.3 0.3>
    gravityOverride {z=-980}
    physicsBodyDefinitions [{
      boundBone {boneName=dyn_bandana_03_l}
      boxExtents=<10 10 10>
      constraintSetup {angularLimitsMax=<0 0 0>, angularLimitsMin=<0 0 -45>, angularTarget=<0 0 0>}
      localJointOffset=<1 0 0>
     }]
    simSpaceSettings {maxAngularAcceleration=1500, maxAngularVelocity=1500, simSpaceAngularAlpha=0.1}
   }
   n35 = Anim Dynamics - Bone: dyn_bandana_01_r(ComponentPose=@n34, Alpha=1, ExternalForce="0,0,0") {
    angularDampingOverride=0.99
    angularSpringConstant=600
    bAngularSpring=true
    bOverrideAngularDamping=true
    bUseGravityOverride=true
    simulationSpace=RootRelative
    boundBone {boneName=dyn_bandana_01_r}
    componentAppliedLinearAccClamp=<1500 1500 2000>
    componentLinearAccScale=<0.3 0.3 0.5>
    componentLinearVelScale=<0.3 0.3 0.2>
    gravityOverride {z=-980}
    physicsBodyDefinitions [{
      boundBone {boneName=dyn_bandana_01_r}
      boxExtents=<10 10 10>
      constraintSetup {angularLimitsMax=<0 0 0>, angularLimitsMin=<0 0 -20>, angularTarget=<0 0 0>}
      localJointOffset=<1 0 0>
     }]
    simSpaceSettings {maxAngularAcceleration=1500, maxAngularVelocity=1500, simSpaceAngularAlpha=0.1}
   }
   n36 = Anim Dynamics - Bone: dyn_bandana_02_r(ComponentPose=@n35, Alpha=1, ExternalForce="0,0,0") {
    angularDampingOverride=0.99
    angularSpringConstant=600
    bAngularSpring=true
    bOverrideAngularDamping=true
    bUseGravityOverride=true
    simulationSpace=RootRelative
    boundBone {boneName=dyn_bandana_02_r}
    componentAppliedLinearAccClamp=<1500 1500 2000>
    componentLinearAccScale=<0.3 0.3 0.5>
    componentLinearVelScale=<0.3 0.3 0.2>
    gravityOverride {z=-980}
    physicsBodyDefinitions [{
      boundBone {boneName=dyn_bandana_02_r}
      boxExtents=<10 10 10>
      constraintSetup {angularLimitsMax=<0 0 0>, angularLimitsMin=<0 0 -30>, angularTarget=<0 0 0>}
      localJointOffset=<1 0 0>
     }]
    simSpaceSettings {maxAngularAcceleration=1500, maxAngularVelocity=1500, simSpaceAngularAlpha=0.1}
   }
   n37 = Anim Dynamics - Bone: dyn_bandana_03_r(ComponentPose=@n36, Alpha=1, ExternalForce="0,0,0") {
    angularDampingOverride=0.99
    angularSpringConstant=600
    bAngularSpring=true
    bOverrideAngularDamping=true
    bUseGravityOverride=true
    simulationSpace=RootRelative
    boundBone {boneName=dyn_bandana_03_r}
    componentAppliedLinearAccClamp=<1500 1500 2000>
    componentLinearAccScale=<0.3 0.3 0.5>
    componentLinearVelScale=<0.3 0.3 0.3>
    gravityOverride {z=-980}
    physicsBodyDefinitions [{
      boundBone {boneName=dyn_bandana_03_r}
      boxExtents=<10 10 10>
      constraintSetup {angularLimitsMax=<0 0 0>, angularLimitsMin=<0 0 -45>, angularTarget=<0 0 0>}
      localJointOffset=<1 0 0>
     }]
    simSpaceSettings {maxAngularAcceleration=1500, maxAngularVelocity=1500, simSpaceAngularAlpha=0.1}
   }
   Save cached pose 'BandanaAnimDynamics'(Pose=Blend Poses by bool(
     BlendPose_0=Component To Local(ComponentPose=@n25)
     BlendPose_1=Component To Local(ComponentPose=@n37)
     bActiveValue=Get BackpackEnabled() -> (BackpackEnabled:bool)
    )) {cacheName=BandanaAnimDynamics}
  ]}
 EventGraph {nodes [
   n1 = Event BlueprintUpdateAnimation()
   > n2 = "Gravity Override State Function (Nativized)"!(
    execute
    CharacterPartAnimInstance:Custom Character Part Anim Instance Object Reference=Self-Reference()
    GravityOverrideParameterSet:Fort Gravity Override Parameter Set Structure=Get Gravity Override Params() -> (Gravity Override Params:Fort Gravity Override Parameter Set Structure)
    RigidBodyAnimNodeParameterSet:Fort Rigid Body Anim Node Parameter Set Structure=Get RBAN Params() -> (RBAN Params:Fort Rigid Body Anim Node Parameter Set Structure)
    DebugDraw:bool=false
    GameTimeInSeconds:float=GetGameTimeInSeconds()
    DebugJointName:Name=root
    StateLogicLayerName:Name=none
   ) -> (
    then
    OutGravityOverride:Vector
    OutCurrentState:Name=None
    OutGravityOverrideParameters:Fort Gravity Override Parameters Structure
    OutRigidBodyAnimNodeParameters:Fort Rigid Body Anim Node Parameters Structure
   )
   > Set Gravity Override(Gravity Override=@n2)
   > Set RBAN Settings Out(RBAN Settings Out=@n2.OutRigidBodyAnimNodeParameters)
   n3 = Get BackpackEnabled()
   > Set IsRigidBodyEnableBackPackNotEquip(IsRigidBodyEnableBackPackNotEquip:bool=AND Boolean(A=Get bRigidBodyEnabled(), B=NOT Boolean(A=@n3))) -> (Output_Get:bool)
   n4 = "Equal (Enum)"(A:EFortWeaponCoreAnimation Enum=Get WeaponCoreAnim!() -> (WeaponCoreAnim), B:EFortWeaponCoreAnimation Enum=Melee)
   > Set FPSUnder30(FPSUnder30:double=float * float(
     A=Remap Float(Input=@n1.DeltaTimeX, OldRange=(X=0.016,Y=0.022), NewRange=(X=0,Y=1), Clamp=true) -> (Output=~)
     B="To Float (Boolean)"(InBool=@n4)
    ) {is=title K2Node_PromotableOperator}) -> (Output_Get:double)
   n5 = Event OnCharacterCustomizationLoaded()
   > Set HeadComponent(HeadComponent:Skeletal Mesh Component Budgeted Object Reference=Get Array Item with EFort Custom Part Type(Array=@n5.CustomMeshParts, PartType=Head)) -> (Output_Get:Skeletal Mesh Component Budgeted Object Reference)
   n6 = CheckForBackbling(self=Self-Reference())
   > Set BackpackEnabled(BackpackEnabled:bool=OR Boolean(A=@n6, B=@n6.bIsBackblingACape)) -> (Output_Get:bool)
   Event BlueprintInitializeAnimation()
  ]}
]
variables [
 BackpackEnabled {type=boolean}
 FPSUnder30 {type=number}
 Gravity Override {type=Vector}
 Gravity Override Params {
  type=FortGravityOverrideParameterSet
  default=(Skydive_GoingUp=(GravityOverride=(WindFrequency=(X=1.500000,Y=7.000000,Z=8.000000),WindAmplitude=(X=200.000000,Y=-500.000000,Z=400.000000),GravityOverride=(X=0.000000,Y=-300.000000,Z=1500.000000),JointName="root")),Skydive_FloatingDown=(GravityOverride=(WindFrequency=(X=1.500000,Y=7.000000,Z=6.000000),WindAmplitude=(X=200.000000,Y=-500.000000,Z=400.000000),WindOffset=(X=0.000000,Y=-600.000000,Z=100.000000),GravityOverride=(X=0.000000,Y=-700.000000,Z=200.000000),JointName="root")),Skydive_DivingDown=(GravityOverride=(WindFrequency=(X=1.500000,Y=12.000000,Z=10.000000),WindAmplitude=(X=50.000000,Y=300.000000,Z=100.000000),GravityOverride=(X=-1500.000000,Y=0.000000,Z=0.000000),JointName="pelvis")),Skydive_Parachuting=(GravityOverride=(WindFrequency=(X=1.500000,Y=7.000000,Z=8.000000),WindAmplitude=(X=200.000000,Y=-500.000000,Z=400.000000),GravityOverride=(X=0.000000,Y=-800.000000,Z=-980.000000),JointName="root")),OnGround_Standing=(GravityOverride=(GravityOverride=(X=0.000000,Y=0.000000,Z=-980.000000),JointName="root")),OnGround_Moving=(GravityOverride=(GravityOverride=(X=0.000000,Y=0.000000,Z=-980.000000),JointName="root")),OnGround_Moving_HighRotationRate=(GravityOverride=(GravityOverride=(X=0.000000,Y=0.000000,Z=-980.000000),JointName="root")),OnGround_Crouch_Moving=(GravityOverride=(GravityOverride=(X=0.000000,Y=0.000000,Z=-980.000000),JointName="root")),Emote_or_Melee=(GravityOverride=(GravityOverride=(X=0.000000,Y=0.000000,Z=-980.000000),JointName="root")),Swimming=(GravityOverride=(GravityOverride=(X=-800.000000,Y=0.000000,Z=0.000000),JointName="pelvis")),Front_End=(GravityOverride=(GravityOverride=(X=0.000000,Y=0.000000,Z=-980.000000),JointName="root")),DBNO=(GravityOverride=(GravityOverride=(X=-980.000000,Y=400.000000,Z=0.000000),JointName="pelvis")),DBNO_Carried=(GravityOverride=(GravityOverride=(X=-980.000000,Y=500.000000,Z=-500.000000),JointName="pelvis")),OnGround_Falling=(GravityOverride=(GravityOverride=(X=0.000000,Y=0.000000,Z=-980.000000),JointName="root")),Ziplining=(GravityOverride=(GravityOverride=(X=0.000000,Y=0.000000,Z=-980.000000),JointName="root")),Skydive_Surfing=(GravityOverride=(WindFrequency=(X=1.500000,Y=7.000000,Z=8.000000),WindAmplitude=(X=200.000000,Y=-500.000000,Z=400.000000),GravityOverride=(X=500.000000,Y=-300.000000,Z=-980.000000),JointName="root")),Riding_Standing=(GravityOverride=(GravityOverride=(X=0.000000,Y=-350.000000,Z=-980.000000),JointName="root")),Riding_Moving=(GravityOverride=(GravityOverride=(X=0.000000,Y=-450.000000,Z=-450.000000),JointName="root")),Riding_Moving_Targeting=(GravityOverride=(GravityOverride=(X=0.000000,Y=-200.000000,Z=-980.000000),JointName="root")),Tactical_Sprint=(GravityOverride=(GravityOverride=(X=0.000000,Y=-450.000000,Z=-500.000000),JointName="root")),Rail_Grinding=(GravityOverride=(GravityOverride=(X=0.000000,Y=0.000000,Z=-980.000000),JointName="root")),)
 }
 HeadComponent {type=SkeletalMeshComponentBudgeted}
 IsRigidBodyEnableBackPackNotEquip {type=boolean}
 RBAN Params {
  type=FortRigidBodyAnimNodeParameterSet
  default=(Skydive_GoingUp=(RigidBodyAnimNode=(SimSpaceSettings=(WorldAlpha=1.000000,DampingAlpha=0.000000,MaxLinearVelocity=15.000000,MaxAngularVelocity=30.000000,MaxLinearAcceleration=600.000000,MaxAngularAcceleration=150.000000,ExternalLinearDragV=(X=0.000000,Y=0.500000,Z=0.000000)))),Skydive_FloatingDown=(RigidBodyAnimNode=(ComponentLinearAccScale=(X=0.000000,Y=0.200000,Z=0.000000),ComponentLinearVelScale=(X=0.000000,Y=0.200000,Z=0.000000),SimSpaceSettings=(WorldAlpha=1.000000,VelocityScaleZ=0.700000,DampingAlpha=0.000000,MaxLinearVelocity=10.000000,MaxAngularVelocity=15.000000,MaxLinearAcceleration=300.000000,MaxAngularAcceleration=90.000000))),Skydive_DivingDown=(RigidBodyAnimNode=(ComponentLinearAccScale=(X=0.100000,Y=0.000000,Z=0.000000),ComponentLinearVelScale=(X=0.100000,Y=0.000000,Z=0.000000),SimSpaceSettings=(WorldAlpha=1.000000,DampingAlpha=0.000000,MaxLinearVelocity=15.000000,MaxAngularVelocity=30.000000,MaxLinearAcceleration=600.000000,MaxAngularAcceleration=150.000000,ExternalLinearDragV=(X=0.000000,Y=0.500000,Z=0.000000)))),Skydive_Parachuting=(RigidBodyAnimNode=(ComponentLinearAccScale=(X=0.000000,Y=0.100000,Z=0.300000),ComponentLinearVelScale=(X=0.000000,Y=0.100000,Z=0.300000),SimSpaceSettings=(WorldAlpha=1.000000,DampingAlpha=0.000000,MaxLinearVelocity=15.000000,MaxAngularVelocity=30.000000,MaxLinearAcceleration=600.000000,MaxAngularAcceleration=150.000000,ExternalLinearDragV=(X=0.000000,Y=0.500000,Z=0.000000)))),OnGround_Standing=(RigidBodyAnimNode=(ComponentLinearAccScale=(X=0.200000,Y=0.400000,Z=0.000000),ComponentLinearVelScale=(X=0.200000,Y=0.400000,Z=0.000000),SimSpaceSettings=(WorldAlpha=1.000000,DampingAlpha=0.000000,MaxLinearVelocity=25.000000,MaxAngularVelocity=10.000000,MaxLinearAcceleration=1200.000000,MaxAngularAcceleration=300.000000,ExternalLinearDragV=(X=0.200000,Y=0.400000,Z=0.200000)))),OnGround_Moving=(RigidBodyAnimNode=(ComponentLinearAccScale=(X=0.200000,Y=0.400000,Z=0.000000),ComponentLinearVelScale=(X=0.200000,Y=0.400000,Z=0.000000),SimSpaceSettings=(WorldAlpha=1.000000,DampingAlpha=0.000000,MaxLinearVelocity=25.000000,MaxAngularVelocity=10.000000,MaxLinearAcceleration=1100.000000,MaxAngularAcceleration=300.000000,ExternalLinearDragV=(X=0.200000,Y=0.400000,Z=0.200000)))),OnGround_Moving_HighRotationRate=(RigidBodyAnimNode=(SimSpaceSettings=(WorldAlpha=1.000000,DampingAlpha=0.000000,MaxLinearVelocity=15.000000,MaxAngularVelocity=30.000000,MaxLinearAcceleration=600.000000,MaxAngularAcceleration=150.000000,ExternalLinearDragV=(X=0.300000,Y=0.500000,Z=0.300000)))),OnGround_Crouch_Moving=(RigidBodyAnimNode=(SimSpaceSettings=(WorldAlpha=1.000000,DampingAlpha=0.000000,MaxLinearVelocity=15.000000,MaxAngularVelocity=30.000000,MaxLinearAcceleration=600.000000,MaxAngularAcceleration=150.000000,ExternalLinearDragV=(X=0.000000,Y=0.500000,Z=0.000000)))),Emote_or_Melee=(RigidBodyAnimNode=(SimSpaceSettings=(WorldAlpha=1.000000,DampingAlpha=0.000000,MaxLinearVelocity=15.000000,MaxAngularVelocity=30.000000,MaxLinearAcceleration=600.000000,MaxAngularAcceleration=150.000000,ExternalLinearDragV=(X=0.000000,Y=0.500000,Z=0.000000)))),Swimming=(RigidBodyAnimNode=(SimSpaceSettings=(WorldAlpha=1.000000,DampingAlpha=0.000000,MaxLinearVelocity=15.000000,MaxAngularVelocity=30.000000,MaxLinearAcceleration=600.000000,MaxAngularAcceleration=150.000000,ExternalLinearDragV=(X=0.000000,Y=0.500000,Z=0.000000)))),Front_End=(RigidBodyAnimNode=(SimSpaceSettings=(WorldAlpha=0.700000,DampingAlpha=0.000000,MaxLinearVelocity=10.000000,MaxAngularVelocity=20.000000,MaxLinearAcceleration=400.000000,MaxAngularAcceleration=90.000000,ExternalLinearDragV=(X=0.000000,Y=0.500000,Z=0.000000)))),DBNO=(RigidBodyAnimNode=(SimSpaceSettings=(WorldAlpha=1.000000,DampingAlpha=0.000000,MaxLinearVelocity=15.000000,MaxAngularVelocity=30.000000,MaxLinearAcceleration=600.000000,MaxAngularAcceleration=150.000000,ExternalLinearDragV=(X=0.000000,Y=0.500000,Z=0.000000)))),DBNO_Carried=(RigidBodyAnimNode=(SimSpaceSettings=(WorldAlpha=1.000000,DampingAlpha=0.000000,MaxLinearVelocity=15.000000,MaxAngularVelocity=30.000000,MaxLinearAcceleration=600.000000,MaxAngularAcceleration=150.000000,ExternalLinearDragV=(X=0.000000,Y=0.500000,Z=0.000000)))),OnGround_Falling=(RigidBodyAnimNode=(SimSpaceSettings=(WorldAlpha=1.000000,DampingAlpha=0.000000,MaxLinearVelocity=15.000000,MaxAngularVelocity=30.000000,MaxLinearAcceleration=600.000000,MaxAngularAcceleration=150.000000,ExternalLinearDragV=(X=0.000000,Y=0.500000,Z=0.000000)))),Ziplining=(RigidBodyAnimNode=(SimSpaceSettings=(WorldAlpha=1.000000,DampingAlpha=0.000000,MaxLinearVelocity=15.000000,MaxAngularVelocity=30.000000,MaxLinearAcceleration=600.000000,MaxAngularAcceleration=150.000000,ExternalLinearDragV=(X=0.000000,Y=0.500000,Z=0.000000)))),Skydive_Surfing=(RigidBodyAnimNode=(SimSpaceSettings=(WorldAlpha=1.000000,DampingAlpha=0.000000,MaxLinearVelocity=15.000000,MaxAngularVelocity=30.000000,MaxLinearAcceleration=600.000000,MaxAngularAcceleration=150.000000,ExternalLinearDragV=(X=0.000000,Y=0.500000,Z=0.000000)))),Riding_Standing=(RigidBodyAnimNode=(SimSpaceSettings=(WorldAlpha=1.000000,DampingAlpha=0.000000,MaxLinearVelocity=15.000000,MaxAngularVelocity=30.000000,MaxLinearAcceleration=600.000000,MaxAngularAcceleration=150.000000,ExternalLinearDragV=(X=0.000000,Y=0.500000,Z=0.000000)))),Riding_Moving=(RigidBodyAnimNode=(SimSpaceSettings=(WorldAlpha=1.000000,DampingAlpha=0.000000,MaxLinearVelocity=15.000000,MaxAngularVelocity=30.000000,MaxLinearAcceleration=600.000000,MaxAngularAcceleration=150.000000,ExternalLinearDragV=(X=0.000000,Y=0.500000,Z=0.000000)))),Riding_Moving_Targeting=(RigidBodyAnimNode=(SimSpaceSettings=(WorldAlpha=1.000000,DampingAlpha=0.000000,MaxLinearVelocity=15.000000,MaxAngularVelocity=30.000000,MaxLinearAcceleration=600.000000,MaxAngularAcceleration=150.000000,ExternalLinearDragV=(X=0.000000,Y=0.500000,Z=0.000000)))),Tactical_Sprint=(RigidBodyAnimNode=(SimSpaceSettings=(WorldAlpha=1.000000,DampingAlpha=0.000000,MaxLinearVelocity=15.000000,MaxAngularVelocity=30.000000,MaxLinearAcceleration=600.000000,MaxAngularAcceleration=150.000000,ExternalLinearDragV=(X=0.000000,Y=0.500000,Z=0.000000)))),Rail_Grinding=(RigidBodyAnimNode=(SimSpaceSettings=(WorldAlpha=0.100000,DampingAlpha=0.000000,MaxLinearVelocity=15.000000,MaxAngularVelocity=30.000000,MaxLinearAcceleration=600.000000,MaxAngularAcceleration=150.000000,ExternalLinearDragV=(X=0.000000,Y=0.500000,Z=0.000000)))),)
 }
 RBAN Settings Out {type=FortRigidBodyAnimNodeParameters}
]
